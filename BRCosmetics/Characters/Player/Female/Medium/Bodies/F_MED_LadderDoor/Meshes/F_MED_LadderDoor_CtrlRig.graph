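# Control Rig for the F_MED_LadderDoor character body that procedurally corrects skirt chain and sandal dyn bones from the body pose.
# Thigh and calf rotations and foot-to-pelvis distances are remapped into pose variables.
# These, with spine and toe angles, weight parent constraints and ModifyTransforms offsets, adjusted by custom idle, skydive, frontend and interpolated motionless-crouch states.
# Center back chain bones are then half-blended toward their left and right counterparts.
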ControlRigBlueprint /BRCosmetics/Characters/Player/Female/Medium/Bodies/F_MED_LadderDoor/Meshes/F_MED_LadderDoor_CtrlRig
graphs [
 RigVMFunctionLibrary {links [], nodes []}
 RigVMModel {nodes [
   RigUnit_BeginExecution = &RigUnit_BeginExecution()
   n1 = &RigVMFunction_MathQuaternionToEuler(
    RotationOrder=ZXY
    Value=&RigUnit_GetRelativeTransformForItem(Child=(Type=Bone,Name="thigh_r"), Parent=(Type=Bone,Name="pelvis"), bChildInitial=False, bParentInitial=False).RelativeTransform.Rotation
   )
   RigVMVariableNode(
    Value=&RigVMFunction_MathDoubleRemap(SourceMaximum=18, SourceMinimum=-1, TargetMaximum=1, TargetMinimum=0, Value=@n1.Result.Y).Result
    ExecuteContext=@RigUnit_BeginExecution.ExecutePin
   ) {variable=R legOut}
   > RigVMVariableNode(Value=&RigVMFunction_MathDoubleRemap(SourceMaximum=-7, SourceMinimum=10, TargetMaximum=1, TargetMinimum=0, Value=@n1.Result.Y).Result) {variable=R thighInward}
   Absolute = &RigVMFunction_MathDoubleAbs(Value=@n1.Result.Z)
   Subtract = &RigVMFunction_MathDoubleSub(A=180, B=@Absolute.Result)
   > RigVMVariableNode(Value=&RigVMFunction_MathDoubleRemap(SourceMaximum=85, SourceMinimum=30, TargetMaximum=1, TargetMinimum=0, Value=@Subtract.Result).Result) {variable=R thighForward}
   n2 = &RigVMFunction_MathQuaternionToEuler(
    RotationOrder=ZXY
    Value=&RigUnit_GetRelativeTransformForItem(Child=(Type=Bone,Name="thigh_l"), Parent=(Type=Bone,Name="pelvis"), bChildInitial=False, bParentInitial=False).RelativeTransform.Rotation
   )
   > RigVMVariableNode(Value=&RigVMFunction_MathDoubleRemap(SourceMaximum=-18, SourceMinimum=1, TargetMaximum=1, TargetMinimum=0, Value=@n2.Result.Y).Result) {variable=L legOut}
   > RigVMVariableNode(Value=&RigVMFunction_MathDoubleRemap(SourceMaximum=-7, SourceMinimum=10, TargetMaximum=0, TargetMinimum=1, Value=@n2.Result.Y).Result) {variable="L thighInward "}
   > RigVMVariableNode(Value=&RigVMFunction_MathDoubleRemap(SourceMaximum=-85, SourceMinimum=-30, TargetMaximum=1, TargetMinimum=0, Value=@n2.Result.Z).Result) {variable=L thighForward}
   n3 = &RigUnit_GetRelativeTransformForItem(Child=(Type=Bone,Name="calf_l"), Parent=(Type=Bone,Name="thigh_l"), bChildInitial=False, bParentInitial=False)
   > RigVMVariableNode(Value=&RigVMFunction_MathDoubleRemap(
     SourceMaximum=60
     SourceMinimum=20
     TargetMaximum=1
     TargetMinimum=0
     Value=&RigVMFunction_MathQuaternionToEuler(RotationOrder=ZXY, Value=@n3.RelativeTransform.Rotation).Result.Z
    ).Result) {variable=L calfBent}
   n4 = &RigUnit_GetRelativeTransformForItem(Child=(Type=Bone,Name="calf_r"), Parent=(Type=Bone,Name="thigh_r"), bChildInitial=False, bParentInitial=False)
   > RigVMVariableNode(Value=&RigVMFunction_MathDoubleRemap(
     SourceMaximum=60
     SourceMinimum=20
     TargetMaximum=1
     TargetMinimum=0
     Value=&RigVMFunction_MathQuaternionToEuler(RotationOrder=ZXY, Value=@n4.RelativeTransform.Rotation).Result.Z
    ).Result) {variable=R calfBent}
   n5 = &RigUnit_GetTransform(Item=(Type=Bone,Name="foot_l"), bInitial=False)
   n6 = &RigUnit_GetTransform(Item=(Type=Bone,Name="dyn_C_chain_back_master"), bInitial=False)
   > RigVMVariableNode(Value=&RigVMFunction_MathDoubleRemap(
     SourceMaximum=34
     SourceMinimum=60
     TargetMaximum=1
     TargetMinimum=0
     Value=&RigVMFunction_MathVectorDistance(B=@n5.Transform.Translation, A=@n6.Transform.Translation).Result
    ).Result) {variable=L footPelvis distance}
   n7 = &RigUnit_GetTransform(Item=(Type=Bone,Name="foot_l"), bInitial=False)
   n8 = &RigUnit_GetTransform(Item=(Type=Bone,Name="pelvis"), bInitial=False)
   > RigVMVariableNode(Value=&RigVMFunction_MathDoubleRemap(
     SourceMaximum=27
     SourceMinimum=34
     TargetMaximum=1
     TargetMinimum=0
     Value=&RigVMFunction_MathVectorDistance(B=@n7.Transform.Translation, A=@n8.Transform.Translation).Result
    ).Result) {variable=L footPelvis distance evenCloser}
   Remap = &RigVMFunction_MathDoubleRemap(
    SourceMaximum=34
    SourceMinimum=60
    TargetMaximum=1
    TargetMinimum=0
    Value=&RigVMFunction_MathVectorDistance(
     B=&RigUnit_GetTransform(Item=(Type=Bone,Name="foot_r"), bInitial=False).Transform.Translation
     A=&RigUnit_GetTransform(Item=(Type=Bone,Name="dyn_C_chain_back_master"), bInitial=False).Transform.Translation
    ).Result
   )
   > RigVMVariableNode(Value=@Remap.Result) {variable=R footPelvis distance}
   n9 = &RigUnit_GetTransform(Item=(Type=Bone,Name="pelvis"), bInitial=False)
   n10 = &RigUnit_GetTransform(Item=(Type=Bone,Name="foot_r"), bInitial=False)
   > RigVMVariableNode(Value=&RigVMFunction_MathDoubleRemap(
     SourceMaximum=27
     SourceMinimum=34
     TargetMaximum=1
     TargetMinimum=0
     Value=&RigVMFunction_MathVectorDistance(A=@n9.Transform.Translation, B=@n10.Transform.Translation).Result
    ).Result) {variable=R footPelvis distance evenCloser}
   n11 = &RigUnit_GetRelativeTransformForItem(Child=(Type=Bone,Name="ball_l"), Parent=(Type=Bone,Name="foot_l"), bChildInitial=False, bParentInitial=False)
   > &RigUnit_ParentConstraint(
    AdvancedSettings=(InterpolationType=Average,RotationOrderForFilter=XZY)
    Child=(Type=Bone,Name="dyn_L_sandal")
    Filter=(TranslationFilter=(bX=false,bY=false,bZ=false),RotationFilter=(bX=true,bY=True,bZ=True),ScaleFilter=(bX=True,bY=True,bZ=True))
    Parents=((Item=(Type=Bone,Name="ball_l"),Weight=1))
    Weight=&RigVMFunction_MathDoubleRemap(
     SourceMaximum=85
     SourceMinimum=90
     TargetMaximum=1
     TargetMinimum=0
     Value=&RigVMFunction_MathQuaternionToEuler(RotationOrder=ZXY, Value=@n11.RelativeTransform.Rotation).Result.Z
    ).Result
   )
   n12 = &RigUnit_GetRelativeTransformForItem(Child=(Type=Bone,Name="ball_r"), Parent=(Type=Bone,Name="foot_r"), bChildInitial=False, bParentInitial=False)
   > &RigUnit_ParentConstraint(
    AdvancedSettings=(InterpolationType=Average,RotationOrderForFilter=XZY)
    Child=(Type=Bone,Name="dyn_R_sandal")
    Filter=(TranslationFilter=(bX=false,bY=false,bZ=false),RotationFilter=(bX=true,bY=True,bZ=True),ScaleFilter=(bX=True,bY=True,bZ=True))
    Parents=((Item=(Type=Bone,Name="ball_r"),Weight=1))
    Weight=&RigVMFunction_MathDoubleRemap(
     SourceMaximum=85
     SourceMinimum=90
     TargetMaximum=1
     TargetMinimum=0
     Value=&RigVMFunction_MathQuaternionToEuler(RotationOrder=ZXY, Value=@n12.RelativeTransform.Rotation).Result.Z
    ).Result
   )
   n13 = RigVMVariableNode() {variable=L footPelvis distance evenCloser}
   > &RigUnit_ParentConstraint(
    AdvancedSettings=(InterpolationType=Average,RotationOrderForFilter=XZY)
    Child=(Type=Bone,Name="dyn_L_chain_back_1")
    Filter=(TranslationFilter=(bX=false,bY=false,bZ=false),RotationFilter=(bX=True,bY=True,bZ=True),ScaleFilter=(bX=True,bY=True,bZ=True))
    Parents=((Item=(Type=Bone,Name="calf_l"),Weight=1))
    Weight=&RigVMFunction_MathDoubleMul(B=0.8, A=&RigVMFunction_MathDoubleSub(A=1, B=@n13.Value).Result).Result
   )
   n14 = RigVMVariableNode() {variable=R footPelvis distance evenCloser}
   > &RigUnit_ParentConstraint(
    AdvancedSettings=(InterpolationType=Average,RotationOrderForFilter=XZY)
    Child=(Type=Bone,Name="dyn_R_chain_back_1")
    Filter=(TranslationFilter=(bX=false,bY=false,bZ=false),RotationFilter=(bX=True,bY=True,bZ=True),ScaleFilter=(bX=True,bY=True,bZ=True))
    Parents=((Item=(Type=Bone,Name="calf_r"),Weight=1))
    Weight=&RigVMFunction_MathDoubleMul(B=0.8, A=&RigVMFunction_MathDoubleSub(A=1, B=@n14.Value).Result).Result
   )
   > &RigUnit_ParentConstraint(
    AdvancedSettings=(InterpolationType=Average,RotationOrderForFilter=XZY)
    Child=(Type=Bone,Name="dyn_L_chain_side_3")
    Filter=(TranslationFilter=(bX=false,bY=false,bZ=false),RotationFilter=(bX=True,bY=True,bZ=True),ScaleFilter=(bX=True,bY=True,bZ=True))
    Parents=((Item=(Type=Bone,Name="calf_l"),Weight=1))
    Weight=&RigVMFunction_MathDoubleSub(A=1, B=RigVMVariableNode() {variable=L footPelvis distance evenCloser}.Value).Result
   )
   > &RigUnit_ParentConstraint(
    AdvancedSettings=(InterpolationType=Average,RotationOrderForFilter=XZY)
    Child=(Type=Bone,Name="dyn_R_chain_side_3")
    Filter=(TranslationFilter=(bX=false,bY=false,bZ=false),RotationFilter=(bX=True,bY=True,bZ=True),ScaleFilter=(bX=True,bY=True,bZ=True))
    Parents=((Item=(Type=Bone,Name="calf_r"),Weight=1))
    Weight=&RigVMFunction_MathDoubleSub(A=1, B=RigVMVariableNode() {variable=R footPelvis distance evenCloser}.Value).Result
   )
   Add = &RigVMFunction_MathDoubleAdd(
    A=&RigVMFunction_MathDoubleSub(A=1, B=RigVMVariableNode() {variable=L thighForward}.Value).Result
    B=RigVMVariableNode() {variable=L legOut}.Value
   )
   n15 = RigVMVariableNode() {variable=L footPelvis distance evenCloser}
   n16 = RigVMVariableNode() {variable=R footPelvis distance evenCloser}
   Multiply = &RigVMFunction_MathDoubleMul(
    B=0.8
    A=&RigVMFunction_MathDoubleSub(
     A=&RigVMFunction_MathDoubleClamp(Value=@Add.Result).Result
     B=&RigVMFunction_MathDoubleMax(A=@n15.Value, B=@n16.Value).Result
    ).Result
   )
   Clamp = &RigVMFunction_MathDoubleClamp(Maximum=0.8, Value=@Multiply.Result)
   > &RigUnit_ParentConstraint(
    AdvancedSettings=(InterpolationType=Average,RotationOrderForFilter=XZY)
    Child=(Type=Bone,Name="dyn_L_chain_back_master")
    Filter=(TranslationFilter=(bX=false,bY=false,bZ=false),RotationFilter=(bX=true,bY=true,bZ=true),ScaleFilter=(bX=True,bY=True,bZ=True))
    Parents=((Item=(Type=Bone,Name="dyn_L_chain_side_1"),Weight=1))
    Weight=@Clamp.Result
   )
   n17 = &RigVMFunction_MathDoubleSub(A=1, B=RigVMVariableNode() {variable=R thighForward}.Value)
   n18 = RigVMVariableNode() {variable=R legOut}
   n19 = &RigVMFunction_MathDoubleSub(
    A=&RigVMFunction_MathDoubleClamp(Value=&RigVMFunction_MathDoubleAdd(A=@n17.Result, B=@n18.Value).Result).Result
    B=&RigVMFunction_MathDoubleMax(
     A=RigVMVariableNode() {variable=L footPelvis distance evenCloser}.Value
     B=RigVMVariableNode() {variable=R footPelvis distance evenCloser}.Value
    ).Result
   )
   > &RigUnit_ParentConstraint(
    AdvancedSettings=(InterpolationType=Average,RotationOrderForFilter=XZY)
    Child=(Type=Bone,Name="dyn_R_chain_back_master")
    Filter=(TranslationFilter=(bX=false,bY=false,bZ=false),RotationFilter=(bX=True,bY=True,bZ=True),ScaleFilter=(bX=True,bY=True,bZ=True))
    Parents=((Item=(Type=Bone,Name="dyn_R_chain_side_1"),Weight=1))
    Weight=&RigVMFunction_MathDoubleClamp(Maximum=0.8, Value=&RigVMFunction_MathDoubleMul(B=0.8, A=@n19.Result).Result).Result
   )
   > &RigUnit_ModifyTransforms(
    ItemToModify="((Item=(Type=Bone,Name=\"dyn_L_chain_back_master\"),Transform=(Rotation=(X=0,Y=-0.1736,Z=0,W=0.9848),Translation=(X=0,Y=0,Z=0),Scale3D=(X=1,Y=1,Z=1))),(Item=(Type=Bone,Name=\"dyn_L_chain_front_master\"),Transform=(Rotation=(X=0,Y=0,Z=0.0436,W=0.999),Translatio ...(truncated)"
    WeightMinimum=0
    Weight=RigVMVariableNode() {variable=L thighForward}.Value
   )
   > &RigUnit_ModifyTransforms(
    ItemToModify="((Item=(Type=Bone,Name=\"dyn_R_chain_back_master\"),Transform=(Rotation=(X=0,Y=-0.1736,Z=0,W=0.9848),Translation=(X=0,Y=0,Z=0),Scale3D=(X=1,Y=1,Z=1))),(Item=(Type=Bone,Name=\"dyn_R_chain_front_master\"),Transform=(Rotation=(X=0,Y=0,Z=0.0436,W=0.999),Translatio ...(truncated)"
    WeightMinimum=0
    Weight=RigVMVariableNode() {variable=R thighForward}.Value
   )
   n20 = RigVMVariableNode() {variable=R thighForward}
   > &RigUnit_ModifyTransforms(
    ItemToModify=((Item=(Type=Bone,Name="dyn_R_chain_front_master"),Transform=(Rotation=(X=0,Y=0,Z=0.4226,W=0.9063),Translation=(X=0,Y=0,Z=0),Scale3D=(X=1,Y=1,Z=1))))
    WeightMinimum=0
    Weight=&RigVMFunction_MathDoubleMul(B=&RigVMFunction_MathDoubleSub(A=1, B=@n20.Value).Result, A=RigVMVariableNode() {variable=L thighForward}.Value).Result
   )
   n21 = RigVMVariableNode() {variable=L thighForward}
   > &RigUnit_ModifyTransforms(
    ItemToModify=((Item=(Type=Bone,Name="dyn_L_chain_front_master"),Transform=(Rotation=(X=0,Y=0,Z=0.4226,W=0.9063),Translation=(X=0,Y=0,Z=0),Scale3D=(X=1,Y=1,Z=1))))
    WeightMinimum=0
    Weight=&RigVMFunction_MathDoubleMul(B=&RigVMFunction_MathDoubleSub(A=1, B=@n21.Value).Result, A=RigVMVariableNode() {variable=R thighForward}.Value).Result
   )
   > &RigUnit_ModifyTransforms(
    ItemToModify=((Item=(Type=Bone,Name="dyn_L_chain_side_master"),Transform=(Rotation=(X=0.0023,Y=-0.0436,Z=0.0523,W=0.9977),Translation=(X=0,Y=0,Z=0),Scale3D=(X=1,Y=1,Z=1))))
    WeightMinimum=0
    Weight=RigVMVariableNode() {variable=L calfBent}.Value
   )
   > &RigUnit_ModifyTransforms(
    ItemToModify=((Item=(Type=Bone,Name="dyn_R_chain_side_master"),Transform=(Rotation=(X=0.0023,Y=-0.0436,Z=0.0523,W=0.9977),Translation=(X=0,Y=0,Z=0),Scale3D=(X=1,Y=1,Z=1))))
    WeightMinimum=0
    Weight=RigVMVariableNode() {variable=R calfBent}.Value
   )
   n22 = RigVMVariableNode() {variable=isSkydive}
   Cast = &RigVMFunction_MathBoolToFloat(Value=RigVMVariableNode() {variable=customIdlePlaying}.Value)
   n23 = RigVMVariableNode() {variable=L legOut}
   > &RigUnit_ModifyTransforms(
    ItemToModify="((Item=(Type=Bone,Name=\"dyn_L_chain_side_1\"),Transform=(Rotation=(X=0,Y=0,Z=0.0436,W=0.999),Translation=(X=0,Y=0,Z=0),Scale3D=(X=1,Y=1,Z=1))),(Item=(Type=Bone,Name=\"dyn_L_chain_front_1\"),Transform=(Rotation=(X=-0.0007,Y=0.0262,Z=0.0262,W=0.9993),Translatio ...(truncated)"
    WeightMinimum=0
    Weight=&RigVMFunction_MathDoubleSub(
     B=&RigVMFunction_MathBoolToFloat(Value=@n22.Value).Result
     A=&RigVMFunction_MathDoubleSub(B=@Cast.Result, A=@n23.Value).Result
    ).Result
   )
   n24 = RigVMVariableNode() {variable=isSkydive}
   n25 = &RigVMFunction_MathBoolToFloat(Value=RigVMVariableNode() {variable=customIdlePlaying}.Value)
   n26 = RigVMVariableNode() {variable=R legOut}
   > &RigUnit_ModifyTransforms(
    ItemToModify="((Item=(Type=Bone,Name=\"dyn_R_chain_side_1\"),Transform=(Rotation=(X=0,Y=0,Z=0.0436,W=0.999),Translation=(X=0,Y=0,Z=0),Scale3D=(X=1,Y=1,Z=1))),(Item=(Type=Bone,Name=\"dyn_R_chain_front_1\"),Transform=(Rotation=(X=-0.0007,Y=0.0262,Z=0.0262,W=0.9993),Translatio ...(truncated)"
    WeightMinimum=0
    Weight=&RigVMFunction_MathDoubleSub(
     B=&RigVMFunction_MathBoolToFloat(Value=@n24.Value).Result
     A=&RigVMFunction_MathDoubleSub(B=@n25.Result, A=@n26.Value).Result
    ).Result
   )
   n27 = &RigVMFunction_MathBoolToFloat(Value=RigVMVariableNode() {variable=customIdlePlaying}.Value)
   n28 = &RigVMFunction_MathDoubleSub(A=1, B=RigVMVariableNode() {variable=L footPelvis distance}.Value)
   > &RigUnit_ModifyTransforms(
    ItemToModify="((Item=(Type=Bone,Name=\"dyn_L_chain_back_master\"),Transform=(Rotation=(X=0,Y=0,Z=0.2588,W=0.9659),Translation=(X=0,Y=0,Z=0),Scale3D=(X=1,Y=1,Z=1))),(Item=(Type=Bone,Name=\"dyn_L_chain_back_2\"),Transform=(Rotation=(X=0,Y=0,Z=0.2588,W=0.9659),Translation=(X=0 ...(truncated)"
    WeightMinimum=0
    Weight=&RigVMFunction_MathDoubleSub(
     A=&RigVMFunction_MathDoubleSub(B=@n27.Result, A=@n28.Result).Result
     B=RigVMVariableNode() {variable=isFrontendNotEmoting}.Value
    ).Result
   )
   n29 = &RigVMFunction_MathBoolToFloat(Value=RigVMVariableNode() {variable=customIdlePlaying}.Value)
   n30 = &RigVMFunction_MathDoubleSub(A=1, B=RigVMVariableNode() {variable=R footPelvis distance}.Value)
   > &RigUnit_ModifyTransforms(
    ItemToModify="((Item=(Type=Bone,Name=\"dyn_R_chain_back_master\"),Transform=(Rotation=(X=0,Y=0,Z=0.2588,W=0.9659),Translation=(X=0,Y=0,Z=0),Scale3D=(X=1,Y=1,Z=1))),(Item=(Type=Bone,Name=\"dyn_R_chain_back_2\"),Transform=(Rotation=(X=0,Y=0,Z=0.2588,W=0.9659),Translation=(X=0 ...(truncated)"
    WeightMinimum=0
    Weight=&RigVMFunction_MathDoubleSub(
     A=&RigVMFunction_MathDoubleSub(B=@n29.Result, A=@n30.Result).Result
     B=RigVMVariableNode() {variable=isFrontendNotEmoting}.Value
    ).Result
   )
   RigVMFunction_MathBoolToFloat = &RigVMFunction_MathBoolToFloat(Value=RigVMVariableNode() {variable=customIdlePlaying}.Value)
   > &RigUnit_ModifyTransforms(
    ItemToModify="((Item=(Type=Bone,Name=\"dyn_L_chain_back_master\"),Transform=(Rotation=(X=0,Y=0,Z=0.0872,W=0.9962),Translation=(X=0,Y=0,Z=-2),Scale3D=(X=1,Y=1,Z=1))),(Item=(Type=Bone,Name=\"dyn_L_chain_side_master\"),Transform=(Rotation=(X=0,Y=0,Z=0.0872,W=0.9962),Translatio ...(truncated)"
    WeightMinimum=0
    Weight=@RigVMFunction_MathBoolToFloat.Result
   )
   AlphaInterp = &RigVMFunction_AlphaInterp(
    InterpSpeedDecreasing=8
    InterpSpeedIncreasing=8
    Scale=1
    bClampResult=False
    Value=&RigVMFunction_MathBoolToFloat(Value=RigVMVariableNode() {variable=isCrouchMotionless}.Value).Result
   )
   > &RigUnit_ModifyTransforms(
    ItemToModify="((Item=(Type=Bone,Name=\"dyn_L_chain_front_3\"),Transform=(Rotation=(X=0,Y=0,Z=-0.5,W=0.866),Translation=(X=0,Y=0,Z=0),Scale3D=(X=1,Y=1,Z=1))),(Item=(Type=Bone,Name=\"dyn_L_chain_front_4\"),Transform=(Rotation=(X=0,Y=0,Z=-0.5,W=0.866),Translation=(X=0,Y=0,Z=0) ...(truncated)"
    WeightMinimum=0
    Weight=@AlphaInterp.Result
   )
   > &RigUnit_ModifyTransforms(
    ItemToModify="((Item=(Type=Bone,Name=\"dyn_L_chain_back_master\"),Transform=(Rotation=(X=0,Y=0,Z=0.342,W=0.9397),Translation=(X=0,Y=0,Z=0),Scale3D=(X=1,Y=1,Z=1))),(Item=(Type=Bone,Name=\"dyn_R_chain_back_master\"),Transform=(Rotation=(X=0,Y=0,Z=0.342,W=0.9397),Translation=( ...(truncated)"
    WeightMinimum=0
    Weight=@AlphaInterp.Result
   )
   n31 = RigVMVariableNode() {variable=R thighInward}
   n32 = RigVMVariableNode() {variable="L thighInward "}
   n33 = RigVMVariableNode() {variable=customIdlePlaying}
   > &RigUnit_ModifyTransforms(
    ItemToModify="((Item=(Type=Bone,Name=\"dyn_L_chain_back_1\"),Transform=(Rotation=(X=0.0227,Y=-0.1722,Z=0.1285,W=0.9764),Translation=(X=0,Y=0,Z=0),Scale3D=(X=1,Y=1,Z=1))),(Item=(Type=Bone,Name=\"dyn_R_chain_back_1\"),Transform=(Rotation=(X=0.0227,Y=-0.1722,Z=0.1285,W=0.9764) ...(truncated)"
    WeightMinimum=0
    Weight=&RigVMFunction_MathDoubleSub(
     A=&RigVMFunction_MathDoubleMul(B=@n31.Value, A=@n32.Value).Result
     B=&RigVMFunction_MathBoolToFloat(Value=@n33.Value).Result
    ).Result
   )
   > &RigUnit_ModifyTransforms(
    ItemToModify=((Item=(Type=Bone,Name="dyn_L_chain_back_master"),Transform=(Rotation=(X=0,Y=0,Z=0.6756,W=0.7373),Translation=(X=9,Y=0,Z=-7),Scale3D=(X=1,Y=1,Z=1))))
    WeightMinimum=0
    Weight=RigVMVariableNode() {variable=L footPelvis distance evenCloser}.Value
   )
   > &RigUnit_ModifyTransforms(
    ItemToModify=((Item=(Type=Bone,Name="dyn_R_chain_back_master"),Transform=(Rotation=(X=0,Y=0,Z=0.6756,W=0.7373),Translation=(X=-9,Y=0,Z=7),Scale3D=(X=1,Y=1,Z=1))))
    WeightMinimum=0
    Weight=RigVMVariableNode() {variable=R footPelvis distance evenCloser}.Value
   )
   > &RigUnit_ModifyTransforms(
    ItemToModify=((Item=(Type=Bone,Name="dyn_C_chain_back_master"),Transform=(Rotation=(X=0,Y=0,Z=0.342,W=0.9397),Translation=(X=9,Y=0,Z=0),Scale3D=(X=1,Y=1,Z=1))))
    WeightMinimum=0
    Weight=&RigVMFunction_MathDoubleMax(
     A=RigVMVariableNode() {variable=L footPelvis distance evenCloser}.Value
     B=RigVMVariableNode() {variable=R footPelvis distance evenCloser}.Value
    ).Result
   )
   n34 = &RigVMFunction_MathDoubleRemap(
    SourceMaximum=-25
    SourceMinimum=0
    TargetMaximum=1
    TargetMinimum=0
    Value=&RigVMFunction_MathQuaternionToEuler(
     RotationOrder=ZXY
     Value=&RigUnit_GetRelativeTransformForItem(Child=(Type=Bone,Name="spine_01"), Parent=(Type=Bone,Name="root"), bChildInitial=False, bParentInitial=False).RelativeTransform.Rotation
    ).Result.Z
   )
   > &RigUnit_ModifyTransforms(
    ItemToModify=((Item=(Type=Bone,Name="dyn_L_chain_back_master"),Transform=(Rotation=(X=0,Y=0,Z=0.2164,W=0.9763),Translation=(X=3,Y=0,Z=-2),Scale3D=(X=1,Y=1,Z=1))))
    WeightMinimum=0
    Weight=&RigVMFunction_MathDoubleMul(B=@n34.Result, A=RigVMVariableNode() {variable=L footPelvis distance evenCloser}.Value).Result
   )
   > &RigUnit_ModifyTransforms(
    ItemToModify=((Item=(Type=Bone,Name="dyn_R_chain_back_master"),Transform=(Rotation=(X=0,Y=0,Z=0.2164,W=0.9763),Translation=(X=-3,Y=0,Z=2),Scale3D=(X=1,Y=1,Z=1))))
    WeightMinimum=0
    Weight=&RigVMFunction_MathDoubleMul(B=@n34.Result, A=RigVMVariableNode() {variable=R footPelvis distance evenCloser}.Value).Result
   )
   n35 = RigVMVariableNode() {variable=L footPelvis distance evenCloser}
   n36 = RigVMVariableNode() {variable=R footPelvis distance evenCloser}
   > &RigUnit_ModifyTransforms(
    ItemToModify=((Item=(Type=Bone,Name="dyn_skirt_ground_collision"),Transform=(Rotation=(X=-0.0436,Y=0,Z=0,W=0.999),Translation=(X=0,Y=0,Z=0),Scale3D=(X=1,Y=1,Z=1))))
    WeightMinimum=0
    Weight=&RigVMFunction_MathDoubleMul(A=&RigVMFunction_MathDoubleMax(A=@n35.Value, B=@n36.Value).Result, B=@n34.Result).Result
   )
   > &RigUnit_ParentConstraint(
    AdvancedSettings=(InterpolationType=Average,RotationOrderForFilter=XZY)
    Child=(Type=Bone,Name="dyn_C_chain_back_master")
    Parents=((Item=(Type=Bone,Name="dyn_L_chain_back_master"),Weight=0.5),(Item=(Type=Bone,Name="dyn_R_chain_back_master"),Weight=0.5))
    Weight=0.5
   )
   > &RigUnit_ParentConstraint(
    AdvancedSettings=(InterpolationType=Average,RotationOrderForFilter=XZY)
    Child=(Type=Bone,Name="dyn_C_chain_back_1")
    Parents=((Item=(Type=Bone,Name="dyn_L_chain_back_1"),Weight=0.5),(Item=(Type=Bone,Name="dyn_R_chain_back_1"),Weight=0.5))
    Weight=0.5
   )
   > &RigUnit_ParentConstraint(
    AdvancedSettings=(InterpolationType=Average,RotationOrderForFilter=XZY)
    Child=(Type=Bone,Name="dyn_C_chain_back_2")
    Parents=((Item=(Type=Bone,Name="dyn_L_chain_back_2"),Weight=0.5),(Item=(Type=Bone,Name="dyn_R_chain_back_2"),Weight=0.5))
    Weight=0.5
   )
   > &RigUnit_ParentConstraint(
    AdvancedSettings=(InterpolationType=Average,RotationOrderForFilter=XZY)
    Child=(Type=Bone,Name="dyn_C_chain_back_3")
    Parents=((Item=(Type=Bone,Name="dyn_L_chain_back_3"),Weight=0.5),(Item=(Type=Bone,Name="dyn_R_chain_back_3"),Weight=0.5))
    Weight=0.5
   )
   > &RigUnit_ParentConstraint(
    AdvancedSettings=(InterpolationType=Average,RotationOrderForFilter=XZY)
    Child=(Type=Bone,Name="dyn_C_chain_back_4")
    Parents=((Item=(Type=Bone,Name="dyn_L_chain_back_4"),Weight=0.5),(Item=(Type=Bone,Name="dyn_R_chain_back_4"),Weight=0.5))
    Weight=0.5
   )
   > &RigUnit_ParentConstraint(
    AdvancedSettings=(InterpolationType=Average,RotationOrderForFilter=XZY)
    Child=(Type=Bone,Name="dyn_C_chain_back_5")
    Parents=((Item=(Type=Bone,Name="dyn_L_chain_back_5"),Weight=0.5),(Item=(Type=Bone,Name="dyn_R_chain_back_5"),Weight=0.5))
    Weight=0.5
   )
   > &RigUnit_ModifyTransforms(
    ItemToModify=((Item=(Type=Bone,Name="dyn_C_chain_back_master"),Transform=(Rotation=(X=0,Y=0,Z=0.342,W=0.9397),Translation=(X=0,Y=0,Z=0),Scale3D=(X=1,Y=1,Z=1))))
    WeightMinimum=0
    Weight=&RigVMFunction_MathDoubleMax(A=RigVMVariableNode() {variable=L footPelvis distance}.Value, B=RigVMVariableNode() {variable=R footPelvis distance}.Value).Result
   )
   &RigVMFunction_MathDoubleRemap(SourceMaximum=-90, SourceMinimum=-50, TargetMaximum=1, TargetMinimum=0, Value=0)
  ]}
]
hierarchy [
 root {loc=<0 0 0>, rot=<0 0 0 1>}
 pelvis {parent=root, loc=<0 -0.5301 98.0571>, rot=<0.0127 0.707 -0.0127 -0.707>}
 spine_01 {parent=pelvis, loc=<6.1208 -0.9328 0>, rot=<0 0 0.034 -0.9994>}
 spine_02 {parent=spine_01, loc=<6.5997 -0.2095 0>, rot=<0 0 0.0238 -0.9997>}
 spine_03 {parent=spine_02, loc=<6.4914 -0.1237 0>, rot=<0 0 0.0149 -0.9999>}
 spine_04 {parent=spine_03, loc=<6.9598 0.0794 0>, rot=<0 0 -0.0132 -0.9999>}
 spine_05 {parent=spine_04, loc=<6.9987 0 0>, rot=<0 0 -0.0015 -1>}
 neck_01 {parent=spine_05, loc=<11.3138 0.1529 0>, rot=<0 0 -0.2003 -0.9797>}
 neck_02 {parent=neck_01, loc=<4.5627 0 0>, rot=<0 0 0.0504 -0.9987>}
 head {parent=neck_02, loc=<5.3914 0 0>, rot=<0 0 0.115 -0.9934>}
 dyn_L_sideStrand_1 {parent=head, loc=<3.6074 0.9022 -8.4634>, rot=<0.0254 -0.7557 0.6482 -0.0905>}
 dyn_L_sideStrand_2 {parent=dyn_L_sideStrand_1, loc=<4.085 0 0>, rot=<0 0 0 1>}
 dyn_R_sideStrand_1 {parent=head, loc=<3.9794 0.6838 8.4091>, rot=<-0.0042 0.6875 0.7261 -0.0104>}
 dyn_R_sideStrand_2 {parent=dyn_R_sideStrand_1, loc=<2.7746 0 0>, rot=<0 0 0 -1>}
 dyn_R_sideStrand_3 {parent=dyn_R_sideStrand_2, loc=<2.9729 0 0>, rot=<0 0 0 1>}
 dyn_L_hair_mstr {parent=head, loc=<19.4694 -12.0925 -2.6354>, rot=<0.4268 -0.903 -0.005 -0.049>}
 dyn_L_hair_1 {parent=dyn_L_hair_mstr, loc=<0 0 0>, rot=<0 0 0 1>}
 dyn_L_hair_2 {parent=dyn_L_hair_1, loc=<11.5654 0 0>, rot=<0 0 -0.3967 -0.918>}
 dyn_L_hair_3 {parent=dyn_L_hair_2, loc=<11.6296 0 0>, rot=<0 0 -0.032 -0.9995>}
 dyn_L_hair_4 {parent=dyn_L_hair_3, loc=<9.6779 0 0>, rot=<0 0 0.2138 -0.9769>}
 dyn_L_hair_5 {parent=dyn_L_hair_4, loc=<10.5554 0 0>, rot=<0 0 0.6525 -0.7578>}
 dyn_R_hair_mstr {parent=head, loc=<19.4694 -12.0925 2.6354>, rot=<0.903 0.4268 -0.049 0.005>}
 dyn_R_hair_1 {parent=dyn_R_hair_mstr, loc=<0 0 0>, rot=<0 0 0 1>}
 dyn_R_hair_2 {parent=dyn_R_hair_1, loc=<-11.5653 0.0001 0>, rot=<0 0 -0.3967 -0.918>}
 dyn_R_hair_3 {parent=dyn_R_hair_2, loc=<-11.6302 0 0.0001>, rot=<0 0 -0.032 -0.9995>}
 dyn_R_hair_4 {parent=dyn_R_hair_3, loc=<-9.6777 0 0>, rot=<0 0 0.2138 -0.9769>}
 dyn_R_hair_5 {parent=dyn_R_hair_4, loc=<-10.5551 0.0001 0>, rot=<0 0 0.6525 -0.7578>}
 dyn_smallBang {parent=head, loc=<11.7915 12.0198 1.0448>, rot=<-0.1788 -0.9503 -0.0492 -0.2501>}
 dyn_largeBang_1 {parent=head, loc=<14.7446 8.7421 -5.2822>, rot=<-0.3042 -0.6426 -0.3995 -0.5788>}
 dyn_largeBang_2 {parent=dyn_largeBang_1, loc=<4.9546 0 0>, rot=<-0.0103 -0.4742 -0.0055 -0.8803>}
 dyn_largeBang_3 {parent=dyn_largeBang_2, loc=<5.0334 0 0>, rot=<-0.0491 -0.2772 -0.0142 -0.9594>}
 dyn_largeBang_4 {parent=dyn_largeBang_3, loc=<5.4099 0 0>, rot=<-0.0001 -0.0423 0.0023 -0.9991>}
 dyn_largeBang_5 {parent=dyn_largeBang_4, loc=<3.0058 0 0>, rot=<0 0 0 1>}
 FACIAL_C_FacialRoot {parent=head, loc=<2.1524 -0.746 -0.0002>, rot=<0.4879 -0.5118 -0.4879 -0.5118>}
 FACIAL_C_Jaw {parent=FACIAL_C_FacialRoot, loc=<0.0099 1.4658 1.5165>, rot=<0.3201 -0.0002 -0.0001 -0.9474>}
 FACIAL_C_Jawline {parent=FACIAL_C_Jaw, loc=<-0.0057 3.0401 8.5696>, rot=<0.2586 0 0 -0.966>}
 FACIAL_L_ChinSide {parent=FACIAL_C_Jaw, loc=<2.361 1.27 8.2418>, rot=<-0.2208 -0.3738 0.0499 -0.8995>}
 FACIAL_R_ChinSide {parent=FACIAL_C_Jaw, loc=<-2.3756 1.2734 8.2428>, rot=<-0.2208 0.3738 -0.0499 -0.8995>}
 FACIAL_C_Chin {parent=FACIAL_C_Jaw, loc=<-0.0069 1.6467 8.6237>, rot=<-0.2591 0 0 -0.9659>}
 FACIAL_C_Forehead {parent=FACIAL_C_FacialRoot, loc=<0.0051 -5.4575 10.3579>, rot=<-0.0667 -0.0002 0 -0.9978>}
 FACIAL_L_ForeheadIn {parent=FACIAL_C_FacialRoot, loc=<2.3129 -5.5973 9.8583>, rot=<-0.0839 -0.0699 0.0026 -0.994>}
 FACIAL_R_ForeheadIn {parent=FACIAL_C_FacialRoot, loc=<-2.3037 -5.5978 9.8525>, rot=<-0.0839 0.0695 -0.0027 -0.994>}
 FACIAL_L_ForeheadMid {parent=FACIAL_C_FacialRoot, loc=<4.4094 -5.9254 8.8544>, rot=<-0.0938 -0.1753 -0.0247 -0.9797>}
 FACIAL_R_ForeheadMid {parent=FACIAL_C_FacialRoot, loc=<-4.3971 -5.9264 8.8481>, rot=<-0.0938 0.1749 0.0247 -0.9798>}
 FACIAL_L_ForeheadOut {parent=FACIAL_C_FacialRoot, loc=<6.3764 -5.7367 7.0862>, rot=<-0.1436 -0.2539 -0.0851 -0.9527>}
 FACIAL_R_ForeheadOut {parent=FACIAL_C_FacialRoot, loc=<-6.3632 -5.7386 7.0826>, rot=<-0.1436 0.2535 0.0851 -0.9528>}
 FACIAL_L_CheekOuter {parent=FACIAL_C_FacialRoot, loc=<5.662 0.3011 5.7643>, rot=<0.0338 -0.3793 0.0915 -0.9201>}
 FACIAL_R_CheekOuter {parent=FACIAL_C_FacialRoot, loc=<-5.6497 0.3034 5.7777>, rot=<0.0337 0.379 -0.0916 -0.9202>}
 FACIAL_L_CheekLower {parent=FACIAL_C_FacialRoot, loc=<4.4063 4.2049 4.9245>, rot=<0.0136 -0.4283 0.0477 -0.9023>}
 FACIAL_R_CheekLower {parent=FACIAL_C_FacialRoot, loc=<-4.4021 4.2042 4.9202>, rot=<0.0135 0.4279 -0.0478 -0.9024>}
 FACIAL_L_Masseter {parent=FACIAL_C_FacialRoot, loc=<7.7004 0.9602 3.1692>, rot=<0.0463 0.1125 0.0104 -0.9925>}
 FACIAL_R_Masseter {parent=FACIAL_C_FacialRoot, loc=<-7.6903 0.9581 3.1757>, rot=<0.0463 -0.1129 -0.0106 -0.9925>}
 FACIAL_L_Ear {parent=FACIAL_C_FacialRoot, loc=<8.2178 -1.6521 -0.4127>, rot=<-0.0496 -0.0003 -0.0262 -0.9984>}
 FACIAL_R_Ear {parent=FACIAL_C_FacialRoot, loc=<-8.1941 -1.6328 -0.3842>, rot=<-0.0496 -0.0001 0.0261 -0.9984>}
 FACIAL_L_Jawline {parent=FACIAL_C_FacialRoot, loc=<4.8324 5.1613 3.7775>, rot=<0.1868 0.1588 -0.0511 -0.9681>}
 FACIAL_R_Jawline {parent=FACIAL_C_FacialRoot, loc=<-4.8268 5.1644 3.7822>, rot=<0.1868 -0.1591 0.0509 -0.9681>}
 FACIAL_C_NoseBridge {parent=FACIAL_C_FacialRoot, loc=<0.0078 -3.4191 11.2432>, rot=<0.0205 -0.0002 -0.0001 -0.9998>}
 FACIAL_L_EyesackUpper {parent=FACIAL_C_FacialRoot, loc=<4.4859 -4.262 9.6594>, rot=<0.0257 -0.0829 -0.0665 -0.994>}
 FACIAL_R_EyesackUpper {parent=FACIAL_C_FacialRoot, loc=<-4.4748 -4.2629 9.6561>, rot=<0.0257 0.0825 0.0664 -0.994>}
 FACIAL_L_EyelidUpperFurrow {parent=FACIAL_C_FacialRoot, loc=<4.0818 -3.8052 9.4657>, rot=<-0.053 -0.1491 -0.0832 -0.9839>}
 FACIAL_R_EyelidUpperFurrow {parent=FACIAL_C_FacialRoot, loc=<-4.0728 -3.8131 9.4652>, rot=<-0.053 0.1487 0.0831 -0.984>}
 FACIAL_L_EyelidUpperA {parent=FACIAL_C_FacialRoot, loc=<3.875 -2.6155 6.2075>, rot=<-0.1708 -0.0002 0 -0.9853>}
 FACIAL_R_EyelidUpperA {parent=FACIAL_C_FacialRoot, loc=<-3.871 -2.6211 6.2113>, rot=<-0.1708 -0.0002 0 -0.9853>}
 FACIAL_L_EyelidLowerA {parent=FACIAL_C_FacialRoot, loc=<3.875 -2.6155 6.2075>, rot=<0.2616 -0.0002 -0.0001 -0.9652>}
 FACIAL_R_EyelidLowerA {parent=FACIAL_C_FacialRoot, loc=<-3.871 -2.6211 6.2113>, rot=<0.2616 -0.0002 -0.0001 -0.9652>}
 FACIAL_L_EyeCornerInner {parent=FACIAL_C_FacialRoot, loc=<2.2882 -1.5558 8.5926>, rot=<-0.0233 -0.0002 -0.0001 -0.9997>}
 FACIAL_R_EyeCornerInner {parent=FACIAL_C_FacialRoot, loc=<-2.28 -1.5591 8.5973>, rot=<-0.0233 -0.0002 -0.0001 -0.9997>}
 FACIAL_L_EyeCornerOuter {parent=FACIAL_C_FacialRoot, loc=<6.4415 -2.5697 6.5108>, rot=<-0.0215 -0.3827 -0.009 -0.9236>}
 FACIAL_R_EyeCornerOuter {parent=FACIAL_C_FacialRoot, loc=<-6.4429 -2.5807 6.555>, rot=<-0.0216 0.3823 0.0089 -0.9237>}
 FACIAL_L_EyesackLower {parent=FACIAL_C_FacialRoot, loc=<4.7678 -0.4698 6.6892>, rot=<-0.0453 -0.1393 -0.0256 -0.9889>}
 FACIAL_R_EyesackLower {parent=FACIAL_C_FacialRoot, loc=<-4.7492 -0.4669 6.696>, rot=<-0.0453 0.139 0.0255 -0.9889>}
 FACIAL_L_CheekInner {parent=FACIAL_C_FacialRoot, loc=<2.9852 0.4399 8.0286>, rot=<-0.1458 -0.0836 -0.1991 -0.9655>}
 FACIAL_R_CheekInner {parent=FACIAL_C_FacialRoot, loc=<-2.9765 0.4387 8.0329>, rot=<-0.1457 0.0832 0.1991 -0.9655>}
 FACIAL_L_NasolabialBulge {parent=FACIAL_C_FacialRoot, loc=<4.1444 4.1732 5.4459>, rot=<0.0364 -0.378 0.015 -0.925>}
 FACIAL_R_NasolabialBulge {parent=FACIAL_C_FacialRoot, loc=<-4.1405 4.172 5.4424>, rot=<0.0363 0.3776 -0.0151 -0.9251>}
 FACIAL_L_NasolabialFurrow {parent=FACIAL_C_FacialRoot, loc=<4.3877 2.9039 8.7744>, rot=<-0.0548 -0.1175 -0.3665 -0.9213>}
 FACIAL_R_NasolabialFurrow {parent=FACIAL_C_FacialRoot, loc=<-4.3835 2.895 8.7779>, rot=<-0.0547 0.1171 0.3664 -0.9214>}
 FACIAL_C_Nose {parent=FACIAL_C_FacialRoot, loc=<0.0063 -0.1008 9.8976>, rot=<-0.1534 -0.0002 0 -0.9882>}
 FACIAL_C_NoseLower {parent=FACIAL_C_Nose, loc=<-0.0021 3.9783 -0.4537>, rot=<0.1305 0 0 -0.9914>}
 FACIAL_L_Nostril {parent=FACIAL_C_Nose, loc=<0.9525 3.4928 -0.4524>, rot=<0.0173 0.1305 0.0023 -0.9913>}
 FACIAL_R_Nostril {parent=FACIAL_C_Nose, loc=<-0.9518 3.4952 -0.4555>, rot=<0.0173 -0.1305 -0.0023 -0.9913>}
 FACIAL_C_NoseTip {parent=FACIAL_C_Nose, loc=<-0.0022 3.0701 0.4483>, rot=<0.0872 0 0 -0.9962>}
 FACIAL_C_MouthUpper {parent=FACIAL_C_FacialRoot, loc=<0.0007 4.5181 6.8077>, rot=<-0.0233 -0.0002 -0.0001 -0.9997>}
 FACIAL_C_LipUpper {parent=FACIAL_C_MouthUpper, loc=<-0.0085 0.4632 3.261>, rot=<-0.0872 0 0 -0.9962>}
 FACIAL_L_LipUpper {parent=FACIAL_C_MouthUpper, loc=<1.356 0.4802 2.9034>, rot=<-0.0778 -0.1601 -0.0379 -0.9833>}
 FACIAL_R_LipUpper {parent=FACIAL_C_MouthUpper, loc=<-1.3658 0.4824 2.8998>, rot=<-0.0778 0.1601 0.0379 -0.9833>}
 FACIAL_L_LipUpperOuter {parent=FACIAL_C_MouthUpper, loc=<2.4796 0.4062 1.9947>, rot=<-0.0984 -0.2293 -0.0983 -0.9634>}
 FACIAL_R_LipUpperOuter {parent=FACIAL_C_MouthUpper, loc=<-2.4798 0.4065 1.9717>, rot=<-0.0984 0.2293 0.0983 -0.9634>}
 FACIAL_L_LipCorner {parent=FACIAL_C_MouthUpper, loc=<2.8287 0.7592 1.1351>, rot=<0 -0.2078 0 -0.9782>}
 FACIAL_R_LipCorner {parent=FACIAL_C_MouthUpper, loc=<-2.8262 0.7623 1.1051>, rot=<0 0.2078 0 -0.9782>}
 FACIAL_C_UnderChin {parent=FACIAL_C_FacialRoot, loc=<0.0046 8.4162 2.7091>, rot=<0.4796 -0.0001 -0.0001 -0.8775>}
 FACIAL_L_UnderChin {parent=FACIAL_C_FacialRoot, loc=<2.6348 7.2029 1.5818>, rot=<0.3829 -0.3568 0.0739 -0.8489>}
 FACIAL_R_UnderChin {parent=FACIAL_C_FacialRoot, loc=<-2.6299 7.2171 1.605>, rot=<0.3828 0.3565 -0.0741 -0.849>}
 FACIAL_C_LowerLipRotation {parent=FACIAL_C_FacialRoot, loc=<0.0099 1.4658 1.5165>, rot=<0.3201 -0.0002 -0.0001 -0.9474>}
 FACIAL_C_MouthLower {parent=FACIAL_C_LowerLipRotation, loc=<-0.0108 -0.782 6.0582>, rot=<-0.342 0 0 -0.9397>}
 FACIAL_C_LipLower {parent=FACIAL_C_MouthLower, loc=<-0.0012 2.1932 2.2902>, rot=<0.0872 0 0 -0.9962>}
 FACIAL_L_LipLower {parent=FACIAL_C_MouthLower, loc=<1.0577 2.0647 2.0262>, rot=<0.0927 -0.1294 0.0272 -0.9869>}
 FACIAL_R_LipLower {parent=FACIAL_C_MouthLower, loc=<-1.0638 2.0651 2.0274>, rot=<0.0927 0.1294 -0.0272 -0.9869>}
 FACIAL_L_LipLowerOuter {parent=FACIAL_C_MouthLower, loc=<2.0943 1.4037 1.3966>, rot=<0.0913 -0.2321 0.0318 -0.9679>}
 FACIAL_R_LipLowerOuter {parent=FACIAL_C_MouthLower, loc=<-2.0994 1.4048 1.3955>, rot=<0.0913 0.2321 -0.0318 -0.9679>}
 FACIAL_C_TeethLower {parent=FACIAL_C_LowerLipRotation, loc=<-0.0203 -1.3682 8.2721>, rot=<-0.342 0 0 -0.9397>}
 FACIAL_C_Tongue1 {parent=FACIAL_C_TeethLower, loc=<0.0161 -0.5407 -5.6534>, rot=<-0.2588 0 0 -0.9659>}
 FACIAL_C_Tongue2 {parent=FACIAL_C_Tongue1, loc=<-0.0048 0.6633 1.5916>, rot=<0.2588 0 0 -0.9659>}
 FACIAL_C_Tongue3 {parent=FACIAL_C_Tongue2, loc=<-0.0009 0.4484 1.446>, rot=<0.1305 0 0 -0.9914>}
 FACIAL_C_Tongue4 {parent=FACIAL_C_Tongue3, loc=<-0.001 -0.0581 1.2922>, rot=<0.0436 0 0 -0.999>}
 FACIAL_C_TongueUpper3 {parent=FACIAL_C_Tongue3, loc=<-0.0007 -0.4534 -0.1233>, rot=<0 0 0 1>}
 FACIAL_C_TongueLower3 {parent=FACIAL_C_Tongue3, loc=<0.0023 0.999 0.0029>, rot=<0 0 0 1>}
 FACIAL_L_TongueSide3 {parent=FACIAL_C_Tongue3, loc=<1.571 0.0737 0.1183>, rot=<0 0 0 1>}
 FACIAL_R_TongueSide3 {parent=FACIAL_C_Tongue3, loc=<-1.5734 0.0771 0.1045>, rot=<0 0 0 1>}
 FACIAL_C_TongueUpper2 {parent=FACIAL_C_Tongue2, loc=<-0.0007 -0.6391 0.049>, rot=<0 0 0 1>}
 FACIAL_L_TongueSide2 {parent=FACIAL_C_Tongue2, loc=<1.8988 0.2726 0.2797>, rot=<0 0 0 1>}
 FACIAL_R_TongueSide2 {parent=FACIAL_C_Tongue2, loc=<-1.9115 0.2684 0.2891>, rot=<0 0 0 1>}
 FACIAL_L_EyelidUpperB {parent=FACIAL_C_FacialRoot, loc=<3.875 -2.6155 6.2075>, rot=<-0.2476 -0.0002 0 -0.9689>}
 FACIAL_L_EyelidLowerB {parent=FACIAL_C_FacialRoot, loc=<3.875 -2.6155 6.2075>, rot=<0.2951 -0.0002 -0.0001 -0.9555>}
 FACIAL_R_EyelidLowerB {parent=FACIAL_C_FacialRoot, loc=<-3.871 -2.6211 6.2113>, rot=<0.3034 -0.0002 -0.0001 -0.9529>}
 FACIAL_R_EyelidUpperB {parent=FACIAL_C_FacialRoot, loc=<-3.871 -2.6211 6.2113>, rot=<-0.2476 -0.0002 0 -0.9689>}
 clavicle_l {parent=spine_05, loc=<5.4831 -0.1053 -3.2946>, rot=<0.0465 -0.7289 -0.009 -0.6829>}
 upperarm_l {parent=clavicle_l, loc=<9.8658 -1.2136 -0.4105>, rot=<0.0164 -0.3245 -0.0269 -0.9454>}
 lowerarm_l {parent=upperarm_l, loc=<26.465 -0.8345 0.259>, rot=<0.0036 0.0161 -0.2298 -0.9731>}
 hand_l {parent=lowerarm_l, loc=<22.7355 -0.3203 -0.6105>, rot=<-0.6784 -0.0223 0.0735 -0.7307>}
 hand_attach_l {parent=hand_l, loc=<8.5 -3 0>, rot=<0.7071 0.7071 0 0>}
 weapon_l {parent=hand_l, loc=<-0.1974 -3.4111 0.3811>, rot=<0 0 0 1>}
 thumb_01_l {parent=hand_l, loc=<1.3118 -0.3191 -2.459>, rot=<0.6546 -0.4042 0.0664 -0.6354>}
 thumb_02_l {parent=thumb_01_l, loc=<5.9336 -0.0448 0.0401>, rot=<-0.0037 0.0213 0.298 -0.9543>}
 thumb_03_l {parent=thumb_02_l, loc=<2.2223 0.3682 -0.0907>, rot=<0.0118 -0.013 0.0066 -0.9998>}
 index_metacarpal_l {parent=hand_l, loc=<1.6535 0.0183 -2.96>, rot=<0.0158 -0.0124 -0.0173 -0.9996>}
 index_01_l {parent=index_metacarpal_l, loc=<8.3008 0.0357 -0.0524>, rot=<-0.0044 0.0485 0.0924 -0.9945>}
 index_02_l {parent=index_01_l, loc=<3.3587 -0.3004 0.0452>, rot=<-0.002 0.0027 0.3225 -0.9466>}
 index_03_l {parent=index_02_l, loc=<2.5578 -0.123 -0.0252>, rot=<-0.002 0.0067 0.2558 -0.9667>}
 middle_metacarpal_l {parent=hand_l, loc=<2.1856 -0.1725 -0.5984>, rot=<0.006 0.0168 -0.0329 -0.9993>}
 middle_01_l {parent=middle_metacarpal_l, loc=<7.3275 -0.2447 0.0129>, rot=<-0.0208 0.0042 0.0858 -0.9961>}
 middle_02_l {parent=middle_01_l, loc=<4.0123 0.1401 0.1175>, rot=<-0.0218 0.0258 0.3512 -0.9357>}
 middle_03_l {parent=middle_02_l, loc=<2.6077 0.0534 0.1373>, rot=<0.0082 0.0188 0.1708 -0.9851>}
 ring_metacarpal_l {parent=hand_l, loc=<2.2074 -0.3625 1.3638>, rot=<-0.0284 0.0722 -0.0458 -0.9959>}
 ring_01_l {parent=ring_metacarpal_l, loc=<7.0295 -0.6197 -0.1446>, rot=<-0.0489 -0.0252 0.1892 -0.9804>}
 ring_02_l {parent=ring_01_l, loc=<3.6174 0.4647 -0.1027>, rot=<0 0 0.2806 -0.9598>}
 ring_03_l {parent=ring_02_l, loc=<2.4152 0.114 0.0199>, rot=<0 0 0.1189 -0.9929>}
 pinky_metacarpal_l {parent=hand_l, loc=<1.7616 -0.9753 3.3814>, rot=<-0.0397 0.0814 -0.069 -0.9935>}
 pinky_01_l {parent=pinky_metacarpal_l, loc=<7.1709 -0.4831 -0.1094>, rot=<-0.0654 -0.0035 0.2242 -0.9723>}
 pinky_02_l {parent=pinky_01_l, loc=<2.8274 0.0893 -0.1043>, rot=<0.0037 0.0011 0.2854 -0.9584>}
 pinky_03_l {parent=pinky_02_l, loc=<2.2046 0.0307 -0.0078>, rot=<-0.0002 0.0098 0.1445 -0.9895>}
 lowerarm_twist_02_l {parent=lowerarm_l, loc=<10.6707 0 0>, rot=<0 0 0 1>}
 lowerarm_twist_01_l {parent=lowerarm_l, loc=<18.316 0 0>, rot=<0 0 0 1>}
 upperarm_twist_01_l {parent=upperarm_l, loc=<7.2135 -0.1134 0>, rot=<0 0 0 1>}
 upperarm_twist_02_l {parent=upperarm_l, loc=<19.7998 -0.1134 0>, rot=<0 0 0 1>}
 clavicle_r {parent=spine_05, loc=<5.4831 -0.1053 3.2946>, rot=<0.7289 0.0465 -0.6829 0.009>}
 upperarm_r {parent=clavicle_r, loc=<-9.8658 1.2136 0.4105>, rot=<0.0164 -0.3245 -0.0269 -0.9454>}
 lowerarm_r {parent=upperarm_r, loc=<-26.465 0.8345 -0.259>, rot=<0.0036 0.0161 -0.2298 -0.9731>}
 hand_r {parent=lowerarm_r, loc=<-22.7355 0.3203 0.6106>, rot=<-0.6784 -0.0223 0.0735 -0.7307>}
 thumb_01_r {parent=hand_r, loc=<-1.3118 0.3191 2.459>, rot=<0.6546 -0.4042 0.0664 -0.6354>}
 thumb_02_r {parent=thumb_01_r, loc=<-5.9336 0.0448 -0.0401>, rot=<-0.0037 0.0213 0.298 -0.9543>}
 thumb_03_r {parent=thumb_02_r, loc=<-2.2223 -0.3682 0.0907>, rot=<0.0118 -0.013 0.0066 -0.9998>}
 index_metacarpal_r {parent=hand_r, loc=<-1.6535 -0.0183 2.96>, rot=<0.0158 -0.0124 -0.0173 -0.9996>}
 index_01_r {parent=index_metacarpal_r, loc=<-8.3008 -0.0357 0.0524>, rot=<-0.0044 0.0485 0.0924 -0.9945>}
 index_02_r {parent=index_01_r, loc=<-3.3587 0.3004 -0.0452>, rot=<-0.002 0.0027 0.3225 -0.9466>}
 index_03_r {parent=index_02_r, loc=<-2.5578 0.123 0.0252>, rot=<-0.002 0.0067 0.2558 -0.9667>}
 middle_metacarpal_r {parent=hand_r, loc=<-2.1856 0.1725 0.5984>, rot=<0.006 0.0168 -0.0329 -0.9993>}
 middle_01_r {parent=middle_metacarpal_r, loc=<-7.3275 0.2447 -0.0129>, rot=<-0.0208 0.0042 0.0858 -0.9961>}
 middle_02_r {parent=middle_01_r, loc=<-4.0123 -0.1401 -0.1175>, rot=<-0.0218 0.0258 0.3512 -0.9357>}
 middle_03_r {parent=middle_02_r, loc=<-2.6077 -0.0534 -0.1373>, rot=<0.0082 0.0188 0.1708 -0.9851>}
 ring_metacarpal_r {parent=hand_r, loc=<-2.2074 0.3625 -1.3638>, rot=<-0.0284 0.0722 -0.0458 -0.9959>}
 ring_01_r {parent=ring_metacarpal_r, loc=<-7.0295 0.6197 0.1446>, rot=<-0.0489 -0.0252 0.1892 -0.9804>}
 ring_02_r {parent=ring_01_r, loc=<-3.6174 -0.4647 0.1027>, rot=<0 0 0.2806 -0.9598>}
 ring_03_r {parent=ring_02_r, loc=<-2.4152 -0.114 -0.0199>, rot=<0 0 0.1189 -0.9929>}
 pinky_metacarpal_r {parent=hand_r, loc=<-1.7616 0.9753 -3.3814>, rot=<-0.0397 0.0814 -0.069 -0.9935>}
 pinky_01_r {parent=pinky_metacarpal_r, loc=<-7.1709 0.4831 0.1094>, rot=<-0.0654 -0.0035 0.2242 -0.9723>}
 pinky_02_r {parent=pinky_01_r, loc=<-2.8274 -0.0893 0.1043>, rot=<0.0037 0.0011 0.2854 -0.9584>}
 pinky_03_r {parent=pinky_02_r, loc=<-2.2046 -0.0307 0.0078>, rot=<-0.0002 0.0098 0.1445 -0.9895>}
 weapon_r {parent=hand_r, loc=<0.1974 3.4111 -0.3811>, rot=<0 0 0 1>}
 hand_attach_r {parent=hand_r, loc=<-8.5001 2.9999 0>, rot=<0 0 -0.7071 -0.7071>}
 lowerarm_twist_02_r {parent=lowerarm_r, loc=<-10.6707 0 0>, rot=<0 0 0 -1>}
 lowerarm_twist_01_r {parent=lowerarm_r, loc=<-18.316 0 0>, rot=<0 0 0 -1>}
 upperarm_twist_01_r {parent=upperarm_r, loc=<-7.2135 0.1134 0>, rot=<0 0 0 -1>}
 upperarm_twist_02_r {parent=upperarm_r, loc=<-19.7998 0.1134 0>, rot=<0 0 0 -1>}
 attach_backpack {parent=spine_05, loc=<-0.123 -13.473 0>, rot=<-0.0093 -0.707 0.0093 -0.707>}
 Backpack_BR {parent=attach_backpack, type=Null, loc=<0 4.6151 0>, rot=<-0.0436 0 0 0.999>}
 attach_cape {parent=spine_05, loc=<9.017 -11.714 0>, rot=<0.0616 -0.7044 -0.0616 -0.7044>}
 cape {parent=attach_cape, type=Null, loc=<0 2.5024 -1.2621>, rot=<0 0 0 1>}
 dyn_L_chest {parent=spine_04, loc=<5.6251 6.3252 -5.7143>, rot=<0.4788 0.5204 0.5204 0.4788>}
 dyn_R_chest {parent=spine_04, loc=<5.6252 6.3252 5.7144>, rot=<-0.5204 0.4788 0.4788 -0.5204>}
 thigh_l {parent=pelvis, loc=<-1.0855 -0.2109 -8.7053>, rot=<0.0009 0.0355 0.0155 -0.9993>}
 calf_l {parent=thigh_l, loc=<-42.5284 -0.0019 -0.7297>, rot=<0.0006 -0.0161 -0.0391 -0.9991>}
 foot_l {parent=calf_l, loc=<-46.7401 -2.5014 -0.592>, rot=<0.0427 -0.0238 0.0353 -0.9982>}
 ball_l {parent=foot_l, loc=<-3.5658 10.8066 -0.1544>, rot=<0.0001 -0.0001 -0.7063 -0.7079>}
 dyn_L_sandal {parent=foot_l, loc=<-7.4886 6.474 -0.1537>, rot=<0.0001 -0.0001 -0.7063 -0.7079>}
 calf_twist_02_l {parent=calf_l, loc=<-19.1546 -1.7943 -0.2638>, rot=<0.0237 0.0095 -0.0035 -0.9997>}
 calf_twist_01_l {parent=calf_l, loc=<-35.8483 -1.9197 -0.5794>, rot=<0.0237 0.0095 -0.0035 -0.9997>}
 thigh_twist_01_l {parent=thigh_l, loc=<-27.952 0 0>, rot=<0.0013 0 0 -1>}
 dyn_L_chain_front_master {parent=thigh_l, loc=<-1.8788 7.7401 -0.1824>, rot=<0.9899 0.0056 -0.022 0.1396>}
 dyn_L_chain_front_1 {parent=dyn_L_chain_front_master, loc=<0 0 0>, rot=<0 0 0 1>}
 dyn_L_chain_front_2 {parent=dyn_L_chain_front_1, loc=<-18 0 0>, rot=<0 0 0 1>}
 dyn_L_chain_front_3 {parent=dyn_L_chain_front_2, loc=<-18 0 0>, rot=<0 0 0 1>}
 dyn_L_chain_front_4 {parent=dyn_L_chain_front_3, loc=<-18 0 0>, rot=<0 0 0 1>}
 dyn_L_chain_front_5 {parent=dyn_L_chain_front_4, loc=<-18 0 0>, rot=<0 0 0 1>}
 dyn_L_chain_side_master {parent=thigh_l, loc=<-2.2558 -0.2096 -3.7955>, rot=<-0.7073 -0.0172 -0.004 -0.7067>}
 dyn_L_chain_side_1 {parent=dyn_L_chain_side_master, loc=<0 0 0>, rot=<0 0 0 1>}
 dyn_L_chain_side_2 {parent=dyn_L_chain_side_1, loc=<-18 0 0>, rot=<0 0 0 1>}
 dyn_L_chain_side_3 {parent=dyn_L_chain_side_2, loc=<-18 0 0>, rot=<0 0 0 1>}
 dyn_L_chain_side_4 {parent=dyn_L_chain_side_3, loc=<-18 0 0>, rot=<0 0 0 1>}
 dyn_L_chain_side_5 {parent=dyn_L_chain_side_4, loc=<-18 0 0>, rot=<0 0 0 1>}
 thigh_r {parent=pelvis, loc=<-1.0855 -0.2109 8.7053>, rot=<0.0355 -0.0009 0.9993 0.0155>}
 calf_r {parent=thigh_r, loc=<42.5284 0.0019 0.7297>, rot=<0.0006 -0.0161 -0.0391 -0.9991>}
 foot_r {parent=calf_r, loc=<46.7401 2.5014 0.592>, rot=<0.0427 -0.0238 0.0353 -0.9982>}
 ball_r {parent=foot_r, loc=<3.5658 -10.8066 0.1544>, rot=<0.0001 -0.0001 -0.7063 -0.7079>}
 dyn_R_sandal {parent=foot_r, loc=<7.4886 -6.474 0.1537>, rot=<0.0001 -0.0001 -0.7063 -0.7079>}
 calf_twist_02_r {parent=calf_r, loc=<19.1546 1.7943 0.2638>, rot=<0.0237 0.0095 -0.0035 -0.9997>}
 calf_twist_01_r {parent=calf_r, loc=<35.8483 1.9197 0.5794>, rot=<0.0237 0.0095 -0.0035 -0.9997>}
 thigh_twist_01_r {parent=thigh_r, loc=<27.952 0 0>, rot=<0.0013 0 0 -1>}
 dyn_R_chain_front_master {parent=thigh_r, loc=<1.8499 -7.7394 -1.3748>, rot=<0.9999 0.0063 -0.0093 -0.0003>}
 dyn_R_chain_front_1 {parent=dyn_R_chain_front_master, loc=<0 0 0>, rot=<0 0 0 1>}
 dyn_R_chain_front_2 {parent=dyn_R_chain_front_1, loc=<18 0 0>, rot=<0 0 0 1>}
 dyn_R_chain_front_A_1 {parent=dyn_R_chain_front_2, loc=<17.8453 -0.0027 5.6006>, rot=<-0.0005 0.0306 0 -0.9995>}
 dyn_R_chain_front_A_2 {parent=dyn_R_chain_front_A_1, loc=<18 0 0>, rot=<0 0 0 1>}
 dyn_R_chain_front_A_3 {parent=dyn_R_chain_front_A_2, loc=<18 0 -0.0001>, rot=<0.0005 0 0 -1>}
 dyn_R_chain_front_B_1 {parent=dyn_R_chain_front_2, loc=<18.383 -1.6019 -3.6001>, rot=<0.2867 -0.0041 -0.0116 -0.9579>}
 dyn_R_chain_front_B_2 {parent=dyn_R_chain_front_B_1, loc=<18 -0.1535 -0.7569>, rot=<0.0999 0 0 -0.995>}
 dyn_R_chain_front_B_3 {parent=dyn_R_chain_front_B_2, loc=<18 0 -0.0001>, rot=<0 0 0 1>}
 dyn_R_chain_side_master {parent=thigh_r, loc=<2.2558 0.2096 3.7955>, rot=<-0.7073 -0.0172 -0.004 -0.7067>}
 dyn_R_chain_side_1 {parent=dyn_R_chain_side_master, loc=<0 0 0>, rot=<0 0 0 -1>}
 dyn_R_chain_side_2 {parent=dyn_R_chain_side_1, loc=<18 0 0>, rot=<0 0 0 -1>}
 dyn_R_chain_side_3 {parent=dyn_R_chain_side_2, loc=<18 -0.0001 0>, rot=<0 0 0 -1>}
 dyn_R_chain_side_4 {parent=dyn_R_chain_side_3, loc=<17.9999 0 0>, rot=<0 0 0 -1>}
 dyn_R_chain_side_5 {parent=dyn_R_chain_side_4, loc=<18 0 0>, rot=<0 0 0 -1>}
 dyn_sash_master {parent=dyn_R_chain_side_1, loc=<9.4333 3.4315 4.4924>, rot=<0.002 -0.0113 -0.0435 -0.999>}
 dyn_sash_1 {parent=dyn_sash_master, loc=<0 0 0>, rot=<0 0 0 1>}
 dyn_sash_2 {parent=dyn_sash_1, loc=<6 0 0>, rot=<0 0 0 1>}
 dyn_sash_3 {parent=dyn_sash_2, loc=<6 0 0>, rot=<-0.0036 -0.0495 0.0335 -0.9982>}
 dyn_sash_4 {parent=dyn_sash_3, loc=<6.0131 0 0>, rot=<-0.0001 -0.0492 0.0005 -0.9988>}
 dyn_sash_5 {parent=dyn_sash_4, loc=<6.0267 0 0>, rot=<-0.0046 -0.052 0.0366 -0.998>}
 dyn_L_chain_back_master {parent=pelvis, loc=<-3.4172 -9.9897 -8.6679>, rot=<-0.1704 0.0087 -0.001 -0.9853>}
 dyn_L_chain_back_1 {parent=dyn_L_chain_back_master, loc=<0 0 0>, rot=<0 0 0 1>}
 dyn_L_chain_back_2 {parent=dyn_L_chain_back_1, loc=<-18 0 0>, rot=<0 0 0 1>}
 dyn_L_chain_back_3 {parent=dyn_L_chain_back_2, loc=<-18 0 0>, rot=<0 0 0 1>}
 dyn_L_chain_back_4 {parent=dyn_L_chain_back_3, loc=<-18 0 0>, rot=<0 0 0 1>}
 dyn_L_chain_back_5 {parent=dyn_L_chain_back_4, loc=<-18 0 0>, rot=<0 0 0 1>}
 dyn_C_chain_back_master {parent=pelvis, loc=<-3.4273 -10.2723 0>, rot=<0 0 0.0001 -1>}
 dyn_C_chain_back_1 {parent=dyn_C_chain_back_master, loc=<0 0 0>, rot=<0 0 0 1>}
 dyn_C_chain_back_2 {parent=dyn_C_chain_back_1, loc=<-18 0 0>, rot=<0 0 0 1>}
 dyn_C_chain_back_3 {parent=dyn_C_chain_back_2, loc=<-18 0 0>, rot=<0 0 0 1>}
 dyn_C_chain_back_4 {parent=dyn_C_chain_back_3, loc=<-18 0 0>, rot=<0 0 0 1>}
 dyn_C_chain_back_5 {parent=dyn_C_chain_back_4, loc=<-18 0 0>, rot=<0 0 0 1>}
 dyn_R_chain_back_master {parent=pelvis, loc=<-3.4172 -9.9897 8.6679>, rot=<0.0087 0.1704 0.9853 -0.001>}
 dyn_R_chain_back_1 {parent=dyn_R_chain_back_master, loc=<0 0 0>, rot=<0 0 0 1>}
 dyn_R_chain_back_2 {parent=dyn_R_chain_back_1, loc=<18 0 0>, rot=<0 0 0 1>}
 dyn_R_chain_back_3 {parent=dyn_R_chain_back_2, loc=<18 0.0001 0>, rot=<0 0 0 1>}
 dyn_R_chain_back_4 {parent=dyn_R_chain_back_3, loc=<18 0 0>, rot=<0 0 0 1>}
 dyn_R_chain_back_5 {parent=dyn_R_chain_back_4, loc=<18.0001 0 0>, rot=<0 0 0 1>}
 attach {parent=root, loc=<0 0 0>, rot=<-0.0173 0.0852 -0.1981 -0.9763>}
 ik_hand_root {parent=root, loc=<0 0 0>, rot=<0 0 0 1>}
 ik_hand_gun {parent=ik_hand_root, loc=<-48.1743 3.3545 104.4381>, rot=<0.6276 0.1309 0.4183 -0.6434>}
 ik_hand_r {parent=ik_hand_gun, loc=<0 0 0>, rot=<0 0 0 1>}
 ik_hand_l {parent=ik_hand_gun, loc=<59.3271 67.6997 34.3524>, rot=<-0.6158 -0.7027 -0.3565 0>}
 ik_foot_root {parent=root, loc=<0 0 0>, rot=<0 0 0 1>}
 ik_foot_l {parent=ik_foot_root, loc=<14.8262 -7.3545 8.296>, rot=<0.0357 0.7045 0.0284 -0.7082>}
 ik_foot_r {parent=ik_foot_root, loc=<-14.8262 -7.3545 8.296>, rot=<0.7082 0.0284 -0.7045 0.0357>}
 root_fp {parent=root, loc=<0 0 0>, rot=<0 0 0 1>}
 attach_fp {parent=root_fp, loc=<0 0 0>, rot=<0 0 0 1>}
 camera_root {parent=root_fp, loc=<0 0 0>, rot=<0 0 0 1>}
 camera {parent=camera_root, loc=<0 0 0>, rot=<0 0 0 1>}
 dyn_skirt_ground_collision {parent=root, loc=<0 0 0>, rot=<0 0 0 1>}
 attach_emote_01 {parent=root, loc=<0 0 0>, rot=<0 0 0 1>}
 attach_emote_02 {parent=root, loc=<0 0 0>, rot=<0 0 0 1>}
 attach_emote_03 {parent=root, loc=<0 0 0>, rot=<0 0 0 1>}
 warp_point_01 {parent=root, loc=<0 0 0>, rot=<0 0 0 1>}
 warp_point_02 {parent=root, loc=<0 0 0>, rot=<0 0 0 1>}
 warp_point_03 {parent=root, loc=<0 0 0>, rot=<0 0 0 1>}
]
variables [
 L calfBent {type=double, default=0}
 L footPelvis distance {type=double, default=0}
 L footPelvis distance evenCloser {type=double, default=0}
 L legOut {type=double, default=0}
 L thighForward {type=double, default=0}
 "L thighInward " {type=double, default=0}
 R calfBent {type=double, default=0}
 R footPelvis distance {type=double, default=0}
 R footPelvis distance evenCloser {type=double, default=0}
 R legOut {type=double, default=0}
 R thighForward {type=double, default=0}
 R thighInward {type=double, default=0}
 customIdlePlaying {type=bool, default=False}
 isCrouchMotionless {type=bool, default=False}
 isFrontendNotEmoting {type=double, default=0}
 isSkydive {type=bool, default=False}
]
