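# Physics asset for the idle mesh of the F_MED_LadderDoor female medium player body, apparently for secondary motion of a hanging garment and sash.
# Back, side and front bone chains use capsules and boxes linked by angle-limited constraints, with twist locked on upper left/right back links.
# The right front splits into A and B branches, and the sash hangs from the right side root.
# Pelvis, thigh and calf capsules have no constraints, suggesting body colliders.
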PhysicsAsset /BRCosmetics/Characters/Player/Female/Medium/Bodies/F_MED_LadderDoor/Meshes/F_MED_LadderDoor_Idle_Physics
bodies {
 calf_l [{shapeType=Capsule, length=44, radius=8, shapeName=calf_r_capsule, center=<-27 0 2.4854>, rotation=<rot 90 0 0>}]
 calf_r [{shapeType=Capsule, length=44, radius=8, shapeName=calf_r_capsule, center=<27 0 -2.4854>, rotation=<rot 90 0 0>}]
 dyn_C_chain_back_1 [{shapeType=Box, extentX=60, extentY=7, extentZ=17, shapeName=dyn_C_chain_back_1_box, center=<-15 -1 0>, rotation=<rot 90 0 0>}]
 dyn_C_chain_back_2 [{shapeType=Box, extentX=65, extentY=7, extentZ=17, shapeName=dyn_C_chain_back_2_box, center=<-15 0 0>, rotation=<rot 90 0 0>}]
 dyn_C_chain_back_3 [{shapeType=Box, extentX=70, extentY=7, extentZ=17, shapeName=dyn_C_chain_back_3_box, center=<-15 0 0>, rotation=<rot 90 0 0>}]
 dyn_C_chain_back_4 [{shapeType=Box, extentX=75, extentY=7, extentZ=17, shapeName=dyn_C_chain_back_4_box, center=<-15 0 0>, rotation=<rot 90 0 0>}]
 dyn_C_chain_back_5 [{shapeType=Box, extentX=80, extentY=7, extentZ=14.9717, shapeName=dyn_C_chain_back_5_box, center=<-8 0 0>, rotation=<rot 90 0 0>}]
 dyn_C_chain_back_master [{shapeType=Capsule, length=0.505, radius=0.505, shapeName=dyn_C_chain_back_master_capsule}]
 dyn_L_chain_back_1 [{shapeType=Capsule, length=15, radius=7, shapeName=dyn_L_chain_back_1_capsule, center=<-8 -4.6 0>}]
 dyn_L_chain_back_2 [{shapeType=Capsule, length=15, radius=7, shapeName=dyn_L_chain_back_1_capsule, center=<-8 -4 0>}]
 dyn_L_chain_back_3 [{shapeType=Box, extentX=30, extentY=6, extentZ=20.9808, shapeName=dyn_L_chain_back_3_box, center=<-10 -0.5 0>, rotation=<rot 90 0 0>}]
 dyn_L_chain_back_4 [{shapeType=Box, extentX=30, extentY=6, extentZ=20.9808, shapeName=dyn_L_chain_back_4_box, center=<-10 0 0>, rotation=<rot 90 0 0>}]
 dyn_L_chain_back_5 [{shapeType=Box, extentX=30, extentY=6, extentZ=14.9717, shapeName=dyn_L_chain_back_5_box, center=<-8 0 0>, rotation=<rot 90 0 0>}]
 dyn_L_chain_back_master [{shapeType=Capsule, length=0.505, radius=0.505, shapeName=dyn_L_chain_back_master_capsule}]
 dyn_L_chain_front_2 [{shapeType=Capsule, length=30, radius=6, shapeName=dyn_L_chain_front_3_capsule, center=<-6.7058 -5.5 0>}]
 dyn_L_chain_front_3 [{shapeType=Capsule, length=15, radius=6, shapeName=dyn_L_chain_front_3_capsule, center=<-6.7058 -5.5 0>}]
 dyn_L_chain_front_4 [{shapeType=Capsule, length=15, radius=6, shapeName=dyn_L_chain_front_4_capsule, center=<-6.7058 -4 0>}]
 dyn_L_chain_front_5 [{shapeType=Capsule, length=15, radius=6, shapeName=dyn_L_chain_front_5_capsule, center=<-8 -2.2 0>}]
 dyn_L_chain_side_1 [{shapeType=Capsule, length=0.505, radius=0.505, shapeName=dyn_L_chain_side_1_capsule}]
 dyn_L_chain_side_2 [{shapeType=Capsule, length=30, radius=6, shapeName=dyn_L_chain_side_2_capsule, center=<-6.7027 -6.5 0>}]
 dyn_L_chain_side_3 [{shapeType=Capsule, length=30, radius=6, shapeName=dyn_L_chain_side_3_capsule, center=<-6.7027 -7 0>}]
 dyn_L_chain_side_4 [{shapeType=Capsule, length=30, radius=6, shapeName=dyn_L_chain_side_4_capsule, center=<-6.7027 -7 0>}]
 dyn_L_chain_side_5 [{shapeType=Capsule, length=30, radius=6, shapeName=dyn_L_chain_side_5_capsule, center=<-8 -7 0>}]
 dyn_R_chain_back_1 [{shapeType=Capsule, length=15, radius=7, shapeName=dyn_R_chain_back_1_capsule, center=<8 4.6 0>}]
 dyn_R_chain_back_2 [{shapeType=Capsule, length=15, radius=7, shapeName=dyn_R_chain_back_1_capsule, center=<8 4 0>}]
 dyn_R_chain_back_3 [{shapeType=Box, extentX=30, extentY=6, extentZ=20.9808, shapeName=dyn_R_chain_back_3_box, center=<10 0.5 0>, rotation=<rot 90 0 0>}]
 dyn_R_chain_back_4 [{shapeType=Box, extentX=30, extentY=6, extentZ=20.9808, shapeName=dyn_R_chain_back_4_box, center=<10 0 0>, rotation=<rot 90 0 0>}]
 dyn_R_chain_back_5 [{shapeType=Box, extentX=30, extentY=6, extentZ=14.9717, shapeName=dyn_R_chain_back_5_box, center=<8 0 0>, rotation=<rot 90 0 0>}]
 dyn_R_chain_back_master [{shapeType=Capsule, length=0.505, radius=0.505, shapeName=dyn_R_chain_back_master_capsule}]
 dyn_R_chain_front_2 [{shapeType=Capsule, length=30, radius=6, shapeName=dyn_R_chain_front_A_1_capsule, center=<6.7147 5.5 0>}]
 dyn_R_chain_front_A_1 [{shapeType=Capsule, length=15, radius=6, shapeName=dyn_R_chain_front_A_1_capsule, center=<6.7147 5.5 0>}]
 dyn_R_chain_front_A_2 [{shapeType=Capsule, length=15, radius=6, shapeName=dyn_R_chain_front_A_2_capsule, center=<6.7147 5.5 0>}]
 dyn_R_chain_front_A_3 [{shapeType=Capsule, length=15, radius=6, shapeName=dyn_R_chain_front_A_3_capsule, center=<8 4 0>}]
 dyn_R_chain_front_B_1 [{shapeType=Capsule, length=15, radius=6, shapeName=dyn_R_chain_front_B_1_capsule, center=<6.701 5.5 0>, rotation=<rot 0 0 -10>}]
 dyn_R_chain_front_B_2 [{shapeType=Capsule, length=15, radius=6, shapeName=dyn_R_chain_front_B_2_capsule, center=<6.701 4.5 0>, rotation=<rot 0 0 -10>}]
 dyn_R_chain_front_B_3 [{shapeType=Capsule, length=15, radius=6, shapeName=dyn_R_chain_front_B_3_capsule, center=<8 2.5 0>, rotation=<rot 0 0 -10>}]
 dyn_R_chain_side_1 [{shapeType=Capsule, length=0.505, radius=0.505, shapeName=dyn_R_chain_side_1_capsule}]
 dyn_R_chain_side_2 [{shapeType=Capsule, length=30, radius=6, shapeName=dyn_R_chain_side_2_capsule, center=<6.7027 6.5 0>}]
 dyn_R_chain_side_3 [{shapeType=Capsule, length=30, radius=6, shapeName=dyn_R_chain_side_3_capsule, center=<6.7027 7 0>}]
 dyn_R_chain_side_4 [{shapeType=Capsule, length=30, radius=6, shapeName=dyn_R_chain_side_4_capsule, center=<6.7027 7 0>}]
 dyn_R_chain_side_5 [{shapeType=Capsule, length=30, radius=6, shapeName=dyn_R_chain_side_5_capsule, center=<8 7 0>}]
 dyn_sash_1 [{shapeType=Capsule, length=7, radius=4, shapeName=dyn_sash_2_capsule, center=<8.729 11 0>, rotation=<rot 90 0 0>}]
 dyn_sash_2 [{shapeType=Capsule, length=7, radius=5, shapeName=dyn_sash_2_capsule, center=<8.729 12.75 0>, rotation=<rot 90 0 0>}]
 dyn_sash_3 [{shapeType=Capsule, length=7, radius=6, shapeName=dyn_sash_3_capsule, center=<8.5367 14 0>, rotation=<rot 90 0 0>}]
 dyn_sash_4 [{shapeType=Capsule, length=7, radius=6, shapeName=dyn_sash_4_capsule, center=<8.5367 13.5 0>, rotation=<rot 90 0 0>}]
 dyn_sash_5 [{shapeType=Capsule, length=7, radius=8, shapeName=dyn_sash_4_capsule, center=<8.5367 17 0>, rotation=<rot 90 0 0>}]
 pelvis [{shapeType=Capsule, length=12, radius=8, shapeName=pelvis_capsule, center=<-8 0.1673 0>}]
 thigh_l [{shapeType=Capsule, length=41, radius=7.8013, shapeName=thigh_r_capsule, center=<-20 0 3>, rotation=<rot 87 0 0>}]
 thigh_r [{shapeType=Capsule, length=41, radius=7.8013, shapeName=thigh_r_capsule, center=<20 0 -3>, rotation=<rot 87 0 0>}]
}
constraints [
 {bone1Name=dyn_C_chain_back_1, bone2Name=dyn_C_chain_back_master, swing1LimitDegrees=15, swing2LimitDegrees=10, twistLimitDegrees=10}
 {bone1Name=dyn_C_chain_back_2, bone2Name=dyn_C_chain_back_1, swing1LimitDegrees=15, swing2LimitDegrees=10, twistLimitDegrees=5}
 {bone1Name=dyn_C_chain_back_3, bone2Name=dyn_C_chain_back_2, swing1LimitDegrees=20, swing2LimitDegrees=20, twistLimitDegrees=20}
 {bone1Name=dyn_C_chain_back_4, bone2Name=dyn_C_chain_back_3, swing1LimitDegrees=20, swing2LimitDegrees=20, twistLimitDegrees=20}
 {bone1Name=dyn_C_chain_back_5, bone2Name=dyn_C_chain_back_4, swing1LimitDegrees=20, swing2LimitDegrees=20, twistLimitDegrees=20}
 {
  bone1Name=dyn_L_chain_back_1
  bone2Name=dyn_L_chain_back_master
  swing1LimitDegrees=15
  swing2LimitDegrees=5
  twistLimitDegrees=20
  twistMotion=Locked
 }
 {
  bone1Name=dyn_L_chain_back_2
  bone2Name=dyn_L_chain_back_1
  swing1LimitDegrees=15
  swing2LimitDegrees=5
  twistLimitDegrees=20
  twistMotion=Locked
 }
 {bone1Name=dyn_L_chain_back_3, bone2Name=dyn_L_chain_back_2, swing1LimitDegrees=20, swing2LimitDegrees=20, twistLimitDegrees=20}
 {bone1Name=dyn_L_chain_back_4, bone2Name=dyn_L_chain_back_3, swing1LimitDegrees=20, swing2LimitDegrees=20, twistLimitDegrees=20}
 {bone1Name=dyn_L_chain_back_5, bone2Name=dyn_L_chain_back_4, swing1LimitDegrees=20, swing2LimitDegrees=20, twistLimitDegrees=20}
 {
  bone1Name=dyn_R_chain_back_1
  bone2Name=dyn_R_chain_back_master
  swing1LimitDegrees=15
  swing2LimitDegrees=5
  twistLimitDegrees=20
  twistMotion=Locked
 }
 {
  bone1Name=dyn_R_chain_back_2
  bone2Name=dyn_R_chain_back_1
  swing1LimitDegrees=15
  swing2LimitDegrees=5
  twistLimitDegrees=20
  twistMotion=Locked
 }
 {bone1Name=dyn_R_chain_back_3, bone2Name=dyn_R_chain_back_2, swing1LimitDegrees=20, swing2LimitDegrees=20, twistLimitDegrees=20}
 {bone1Name=dyn_R_chain_back_4, bone2Name=dyn_R_chain_back_3, swing1LimitDegrees=20, swing2LimitDegrees=20, twistLimitDegrees=20}
 {bone1Name=dyn_R_chain_back_5, bone2Name=dyn_R_chain_back_4, swing1LimitDegrees=20, swing2LimitDegrees=20, twistLimitDegrees=20}
 {bone1Name=dyn_L_chain_front_3, bone2Name=dyn_L_chain_front_2, swing1LimitDegrees=20, swing2LimitDegrees=20, twistLimitDegrees=20}
 {bone1Name=dyn_L_chain_front_4, bone2Name=dyn_L_chain_front_3, swing1LimitDegrees=20, swing2LimitDegrees=20, twistLimitDegrees=20}
 {bone1Name=dyn_L_chain_front_5, bone2Name=dyn_L_chain_front_4, swing1LimitDegrees=20, swing2LimitDegrees=20, twistLimitDegrees=20}
 {bone1Name=dyn_L_chain_side_2, bone2Name=dyn_L_chain_side_1, swing1LimitDegrees=20, swing2LimitDegrees=20, twistLimitDegrees=20}
 {bone1Name=dyn_L_chain_side_3, bone2Name=dyn_L_chain_side_2, swing1LimitDegrees=20, swing2LimitDegrees=20, twistLimitDegrees=20}
 {bone1Name=dyn_L_chain_side_4, bone2Name=dyn_L_chain_side_3, swing1LimitDegrees=20, swing2LimitDegrees=20, twistLimitDegrees=20}
 {bone1Name=dyn_L_chain_side_5, bone2Name=dyn_L_chain_side_4, swing1LimitDegrees=20, swing2LimitDegrees=20, twistLimitDegrees=20}
 {bone1Name=dyn_R_chain_front_A_1, bone2Name=dyn_R_chain_front_2, swing1LimitDegrees=20, swing2LimitDegrees=20, twistLimitDegrees=20}
 {bone1Name=dyn_R_chain_front_A_2, bone2Name=dyn_R_chain_front_A_1, swing1LimitDegrees=20, swing2LimitDegrees=20, twistLimitDegrees=20}
 {bone1Name=dyn_R_chain_front_A_3, bone2Name=dyn_R_chain_front_A_2, swing1LimitDegrees=20, swing2LimitDegrees=20, twistLimitDegrees=20}
 {bone1Name=dyn_R_chain_front_B_1, bone2Name=dyn_R_chain_front_2, swing1LimitDegrees=20, swing2LimitDegrees=20, twistLimitDegrees=20}
 {bone1Name=dyn_R_chain_front_B_2, bone2Name=dyn_R_chain_front_B_1, swing1LimitDegrees=20, swing2LimitDegrees=20, twistLimitDegrees=20}
 {bone1Name=dyn_R_chain_front_B_3, bone2Name=dyn_R_chain_front_B_2, swing1LimitDegrees=20, swing2LimitDegrees=20, twistLimitDegrees=20}
 {bone1Name=dyn_R_chain_side_2, bone2Name=dyn_R_chain_side_1, swing1LimitDegrees=20, swing2LimitDegrees=20, twistLimitDegrees=20}
 {bone1Name=dyn_R_chain_side_3, bone2Name=dyn_R_chain_side_2, swing1LimitDegrees=20, swing2LimitDegrees=20, twistLimitDegrees=20}
 {bone1Name=dyn_R_chain_side_4, bone2Name=dyn_R_chain_side_3, swing1LimitDegrees=20, swing2LimitDegrees=20, twistLimitDegrees=20}
 {bone1Name=dyn_R_chain_side_5, bone2Name=dyn_R_chain_side_4, swing1LimitDegrees=20, swing2LimitDegrees=20, twistLimitDegrees=20}
 {bone1Name=dyn_sash_1, bone2Name=dyn_R_chain_side_1, swing1LimitDegrees=30, swing2LimitDegrees=20, twistLimitDegrees=10}
 {bone1Name=dyn_sash_2, bone2Name=dyn_sash_1, swing1LimitDegrees=30, swing2LimitDegrees=20, twistLimitDegrees=10}
 {bone1Name=dyn_sash_3, bone2Name=dyn_sash_2, swing1LimitDegrees=30, swing2LimitDegrees=20, twistLimitDegrees=10}
 {bone1Name=dyn_sash_4, bone2Name=dyn_sash_3, swing1LimitDegrees=30, swing2LimitDegrees=20, twistLimitDegrees=10}
 {bone1Name=dyn_sash_5, bone2Name=dyn_sash_4, swing1LimitDegrees=30, swing2LimitDegrees=20, twistLimitDegrees=10}
]
