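# Animation Blueprint for the female medium LadderDoor body cosmetic, adding physics-driven motion to its front, side and back chain bones.
# RigidBody nodes simulate the cached startPose with separate skydive, default and custom-idle physics assets, selected by bIsSkyDiving and CustomIdleIsPlayingBool.
# Additive bone rotations, offsets and scales adjust the chains on the skydive and non-idle branches.
# A long per-bone AnimDynamics series is present but unconnected to the output pose.
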AnimBlueprint /BRCosmetics/Characters/Player/Female/Medium/Bodies/F_MED_LadderDoor/Meshes/F_MED_LadderDoor_AnimBP
target_skeleton=&/F_MED_LadderDoor_Skeleton
graphs [
 AnimGraph {nodes [
   n1 = Anim Dynamics - Bone: dyn_L_chain_front_1(Alpha=1, ExternalForce="0,0,0") {
    angularDampingOverride=1
    angularSpringConstant=10000
    bAngularSpring=true
    bOverrideAngularDamping=true
    bUseGravityOverride=true
    simulationSpace=RootRelative
    alphaScaleBiasClamp {bias=1.2, scale=-1}
    boundBone {boneName=dyn_L_chain_front_1}
    gravityOverride {z=-980}
    physicsBodyDefinitions [{
      boundBone {boneName=dyn_L_chain_front_1}
      boxExtents=<25 25 25>
      constraintSetup {angularLimitsMax=<0 2 20>, angularLimitsMin=<0 -2 0>, angularTarget=<0 0 0>}
      localJointOffset=<-4 0 0>
     }]
    simSpaceSettings {maxAngularAcceleration=4000, maxAngularVelocity=8, simSpaceAngularAlpha=1}
   }
   n2 = Anim Dynamics - Bone: dyn_L_chain_front_2(ComponentPose=@n1, Alpha=1, ExternalForce="0,0,0") {
    angularDampingOverride=1
    angularSpringConstant=10000
    bAngularSpring=true
    bOverrideAngularDamping=true
    bUseGravityOverride=true
    simulationSpace=RootRelative
    alphaScaleBiasClamp {bias=1.2, scale=-1}
    boundBone {boneName=dyn_L_chain_front_2}
    gravityOverride {z=-980}
    physicsBodyDefinitions [{
      boundBone {boneName=dyn_L_chain_front_2}
      boxExtents=<25 25 25>
      constraintSetup {angularLimitsMax=<0 2 20>, angularLimitsMin=<0 -2 0>, angularTarget=<0 0 0>}
      localJointOffset=<-4 0 0>
     }]
    simSpaceSettings {maxAngularAcceleration=4000, maxAngularVelocity=8, simSpaceAngularAlpha=1}
   }
   n3 = Anim Dynamics - Bone: dyn_L_chain_front_3(ComponentPose=@n2, Alpha=1, ExternalForce="0,0,0") {
    angularDampingOverride=1
    angularSpringConstant=10000
    bAngularSpring=true
    bOverrideAngularDamping=true
    bUseGravityOverride=true
    simulationSpace=RootRelative
    alphaScaleBiasClamp {bias=1.2, scale=-1}
    boundBone {boneName=dyn_L_chain_front_3}
    gravityOverride {z=-980}
    physicsBodyDefinitions [{
      boundBone {boneName=dyn_L_chain_front_3}
      boxExtents=<25 25 25>
      constraintSetup {angularLimitsMax=<0 2 20>, angularLimitsMin=<0 -2 0>, angularTarget=<0 0 0>}
      localJointOffset=<-4 0 0>
     }]
    simSpaceSettings {maxAngularAcceleration=4000, maxAngularVelocity=8, simSpaceAngularAlpha=1}
   }
   n4 = Anim Dynamics - Bone: dyn_L_chain_front_4(ComponentPose=@n3, Alpha=1, ExternalForce="0,0,0") {
    angularDampingOverride=1
    angularSpringConstant=10000
    bAngularSpring=true
    bOverrideAngularDamping=true
    bUseGravityOverride=true
    simulationSpace=RootRelative
    alphaScaleBiasClamp {bias=1.2, scale=-1}
    boundBone {boneName=dyn_L_chain_front_4}
    gravityOverride {z=-980}
    physicsBodyDefinitions [{
      boundBone {boneName=dyn_L_chain_front_4}
      boxExtents=<25 25 25>
      constraintSetup {angularLimitsMax=<0 2 20>, angularLimitsMin=<0 -2 -15>, angularTarget=<0 0 0>}
      localJointOffset=<-4 0 0>
     }]
    simSpaceSettings {maxAngularAcceleration=4000, maxAngularVelocity=8, simSpaceAngularAlpha=1}
   }
   n5 = Anim Dynamics - Bone: dyn_L_chain_front_5(ComponentPose=@n4, Alpha=1, ExternalForce="0,0,0") {
    angularDampingOverride=1
    angularSpringConstant=10000
    bAngularSpring=true
    bOverrideAngularDamping=true
    bUseGravityOverride=true
    simulationSpace=RootRelative
    alphaScaleBiasClamp {bias=1.2, scale=-1}
    boundBone {boneName=dyn_L_chain_front_5}
    gravityOverride {z=-980}
    physicsBodyDefinitions [{
      boundBone {boneName=dyn_L_chain_front_5}
      boxExtents=<25 25 25>
      constraintSetup {angularLimitsMax=<0 2 20>, angularLimitsMin=<0 -2 -30>, angularTarget=<0 0 0>}
      localJointOffset=<-4 0 0>
     }]
    simSpaceSettings {maxAngularAcceleration=4000, maxAngularVelocity=8, simSpaceAngularAlpha=1}
   }
   n6 = Anim Dynamics - Bone: dyn_R_chain_front_1(ComponentPose=@n5, Alpha=1, ExternalForce="0,0,0") {
    angularDampingOverride=1
    angularSpringConstant=10000
    bAngularSpring=true
    bOverrideAngularDamping=true
    bUseGravityOverride=true
    simulationSpace=RootRelative
    alphaScaleBiasClamp {bias=1.2, scale=-1}
    boundBone {boneName=dyn_R_chain_front_1}
    gravityOverride {z=-980}
    physicsBodyDefinitions [{
      boundBone {boneName=dyn_R_chain_front_1}
      boxExtents=<25 25 25>
      constraintSetup {angularLimitsMax=<0 2 20>, angularLimitsMin=<0 -2 0>, angularTarget=<0 0 0>}
      localJointOffset=<4 0 0>
     }]
    simSpaceSettings {maxAngularAcceleration=4000, maxAngularVelocity=8, simSpaceAngularAlpha=1}
   }
   n7 = Anim Dynamics - Bone: dyn_R_chain_front_2(ComponentPose=@n6, Alpha=1, ExternalForce="0,0,0") {
    angularDampingOverride=1
    angularSpringConstant=10000
    bAngularSpring=true
    bOverrideAngularDamping=true
    bUseGravityOverride=true
    simulationSpace=RootRelative
    alphaScaleBiasClamp {bias=1.2, scale=-1}
    boundBone {boneName=dyn_R_chain_front_2}
    gravityOverride {z=-980}
    physicsBodyDefinitions [{
      boundBone {boneName=dyn_R_chain_front_2}
      boxExtents=<25 25 25>
      constraintSetup {angularLimitsMax=<0 2 20>, angularLimitsMin=<0 -2 0>, angularTarget=<0 0 0>}
      localJointOffset=<4 0 0>
     }]
    simSpaceSettings {maxAngularAcceleration=4000, maxAngularVelocity=8, simSpaceAngularAlpha=1}
   }
   n8 = Anim Dynamics - Bone: dyn_R_chain_front_A_1(ComponentPose=@n7, Alpha=1, ExternalForce="0,0,0") {
    angularDampingOverride=1
    angularSpringConstant=10000
    bAngularSpring=true
    bOverrideAngularDamping=true
    bUseGravityOverride=true
    simulationSpace=RootRelative
    alphaScaleBiasClamp {bias=1.2, scale=-1}
    boundBone {boneName=dyn_R_chain_front_A_1}
    gravityOverride {z=-980}
    physicsBodyDefinitions [{
      boundBone {boneName=dyn_R_chain_front_A_1}
      boxExtents=<25 25 25>
      constraintSetup {angularLimitsMax=<0 2 20>, angularLimitsMin=<0 -2 0>, angularTarget=<0 0 0>}
      localJointOffset=<4 0 0>
     }]
    simSpaceSettings {maxAngularAcceleration=4000, maxAngularVelocity=8, simSpaceAngularAlpha=1}
   }
   n9 = Anim Dynamics - Bone: dyn_R_chain_front_A_2(ComponentPose=@n8, Alpha=1, ExternalForce="0,0,0") {
    angularDampingOverride=1
    angularSpringConstant=10000
    bAngularSpring=true
    bOverrideAngularDamping=true
    bUseGravityOverride=true
    simulationSpace=RootRelative
    alphaScaleBiasClamp {bias=1.2, scale=-1}
    boundBone {boneName=dyn_R_chain_front_A_2}
    gravityOverride {z=-980}
    physicsBodyDefinitions [{
      boundBone {boneName=dyn_R_chain_front_A_2}
      boxExtents=<25 25 25>
      constraintSetup {angularLimitsMax=<0 2 20>, angularLimitsMin=<0 -2 -15>, angularTarget=<0 0 0>}
      localJointOffset=<4 0 0>
     }]
    simSpaceSettings {maxAngularAcceleration=4000, maxAngularVelocity=8, simSpaceAngularAlpha=1}
   }
   n10 = Anim Dynamics - Bone: dyn_R_chain_front_A_3(ComponentPose=@n9, Alpha=1, ExternalForce="0,0,0") {
    angularDampingOverride=1
    angularSpringConstant=10000
    bAngularSpring=true
    bOverrideAngularDamping=true
    bUseGravityOverride=true
    simulationSpace=RootRelative
    alphaScaleBiasClamp {bias=1.2, scale=-1}
    boundBone {boneName=dyn_R_chain_front_A_3}
    gravityOverride {z=-980}
    physicsBodyDefinitions [{
      boundBone {boneName=dyn_R_chain_front_A_3}
      boxExtents=<25 25 25>
      constraintSetup {angularLimitsMax=<0 2 20>, angularLimitsMin=<0 -2 -30>, angularTarget=<0 0 0>}
      localJointOffset=<4 0 0>
     }]
    simSpaceSettings {maxAngularAcceleration=4000, maxAngularVelocity=8, simSpaceAngularAlpha=1}
   }
   n11 = Anim Dynamics - Bone: dyn_R_chain_front_B_1(ComponentPose=@n10, Alpha=1, ExternalForce="0,0,0") {
    angularDampingOverride=1
    angularSpringConstant=10000
    bAngularSpring=true
    bOverrideAngularDamping=true
    bUseGravityOverride=true
    simulationSpace=RootRelative
    alphaScaleBiasClamp {bias=1.2, scale=-1}
    boundBone {boneName=dyn_R_chain_front_B_1}
    gravityOverride {z=-980}
    physicsBodyDefinitions [{
      boundBone {boneName=dyn_R_chain_front_B_1}
      boxExtents=<25 25 25>
      constraintSetup {angularLimitsMax=<0 2 20>, angularLimitsMin=<0 -2 0>, angularTarget=<0 0 0>}
      localJointOffset=<4 0 0>
     }]
    simSpaceSettings {maxAngularAcceleration=4000, maxAngularVelocity=8, simSpaceAngularAlpha=1}
   }
   n12 = Anim Dynamics - Bone: dyn_R_chain_front_B_2(ComponentPose=@n11, Alpha=1, ExternalForce="0,0,0") {
    angularDampingOverride=1
    angularSpringConstant=10000
    bAngularSpring=true
    bOverrideAngularDamping=true
    bUseGravityOverride=true
    simulationSpace=RootRelative
    alphaScaleBiasClamp {bias=1.2, scale=-1}
    boundBone {boneName=dyn_R_chain_front_B_2}
    gravityOverride {z=-980}
    physicsBodyDefinitions [{
      boundBone {boneName=dyn_R_chain_front_B_2}
      boxExtents=<25 25 25>
      constraintSetup {angularLimitsMax=<0 2 20>, angularLimitsMin=<0 -2 -15>, angularTarget=<0 0 0>}
      localJointOffset=<4 0 0>
     }]
    simSpaceSettings {maxAngularAcceleration=4000, maxAngularVelocity=8, simSpaceAngularAlpha=1}
   }
   n13 = Anim Dynamics - Bone: dyn_R_chain_front_B_3(ComponentPose=@n12, Alpha=1, ExternalForce="0,0,0") {
    angularDampingOverride=1
    angularSpringConstant=10000
    bAngularSpring=true
    bOverrideAngularDamping=true
    bUseGravityOverride=true
    simulationSpace=RootRelative
    alphaScaleBiasClamp {bias=1.2, scale=-1}
    boundBone {boneName=dyn_R_chain_front_B_3}
    gravityOverride {z=-980}
    physicsBodyDefinitions [{
      boundBone {boneName=dyn_R_chain_front_B_3}
      boxExtents=<25 25 25>
      constraintSetup {angularLimitsMax=<0 2 20>, angularLimitsMin=<0 -2 -30>, angularTarget=<0 0 0>}
      localJointOffset=<4 0 0>
     }]
    simSpaceSettings {maxAngularAcceleration=4000, maxAngularVelocity=8, simSpaceAngularAlpha=1}
   }
   n14 = Anim Dynamics - Bone: dyn_L_chain_side_1(ComponentPose=@n13, Alpha=1, ExternalForce="0,0,0") {
    angularDampingOverride=1
    angularSpringConstant=10000
    bAngularSpring=true
    bOverrideAngularDamping=true
    bUseGravityOverride=true
    simulationSpace=RootRelative
    alphaScaleBiasClamp {bias=1.2, scale=-1}
    boundBone {boneName=dyn_L_chain_side_1}
    gravityOverride {z=-980}
    physicsBodyDefinitions [{
      boundBone {boneName=dyn_L_chain_side_1}
      boxExtents=<25 25 25>
      constraintSetup {angularLimitsMax=<0 2 20>, angularLimitsMin=<0 -2 0>, angularTarget=<0 0 0>}
      localJointOffset=<-4 0 0>
     }]
    simSpaceSettings {maxAngularAcceleration=4000, maxAngularVelocity=8, simSpaceAngularAlpha=1}
   }
   n15 = Anim Dynamics - Bone: dyn_L_chain_side_2(ComponentPose=@n14, Alpha=1, ExternalForce="0,0,0") {
    angularDampingOverride=1
    angularSpringConstant=10000
    bAngularSpring=true
    bOverrideAngularDamping=true
    bUseGravityOverride=true
    simulationSpace=RootRelative
    alphaScaleBiasClamp {bias=1.2, scale=-1}
    boundBone {boneName=dyn_L_chain_side_2}
    gravityOverride {z=-980}
    physicsBodyDefinitions [{
      boundBone {boneName=dyn_L_chain_side_2}
      boxExtents=<25 25 25>
      constraintSetup {angularLimitsMax=<0 2 20>, angularLimitsMin=<0 -2 0>, angularTarget=<0 0 0>}
      localJointOffset=<-4 0 0>
     }]
    simSpaceSettings {maxAngularAcceleration=4000, maxAngularVelocity=8, simSpaceAngularAlpha=1}
   }
   n16 = Anim Dynamics - Bone: dyn_L_chain_side_3(ComponentPose=@n15, Alpha=1, ExternalForce="0,0,0") {
    angularDampingOverride=1
    angularSpringConstant=10000
    bAngularSpring=true
    bOverrideAngularDamping=true
    bUseGravityOverride=true
    simulationSpace=RootRelative
    alphaScaleBiasClamp {bias=1.2, scale=-1}
    boundBone {boneName=dyn_L_chain_side_3}
    gravityOverride {z=-980}
    physicsBodyDefinitions [{
      boundBone {boneName=dyn_L_chain_side_3}
      boxExtents=<25 25 25>
      constraintSetup {angularLimitsMax=<0 2 20>, angularLimitsMin=<0 -2 0>, angularTarget=<0 0 0>}
      localJointOffset=<-4 0 0>
     }]
    simSpaceSettings {maxAngularAcceleration=4000, maxAngularVelocity=8, simSpaceAngularAlpha=1}
   }
   n17 = Anim Dynamics - Bone: dyn_L_chain_side_4(ComponentPose=@n16, Alpha=1, ExternalForce="0,0,0") {
    angularDampingOverride=1
    angularSpringConstant=10000
    bAngularSpring=true
    bOverrideAngularDamping=true
    bUseGravityOverride=true
    simulationSpace=RootRelative
    alphaScaleBiasClamp {bias=1.2, scale=-1}
    boundBone {boneName=dyn_L_chain_side_4}
    gravityOverride {z=-980}
    physicsBodyDefinitions [{
      boundBone {boneName=dyn_L_chain_side_4}
      boxExtents=<25 25 25>
      constraintSetup {angularLimitsMax=<0 2 20>, angularLimitsMin=<0 -2 0>, angularTarget=<0 0 0>}
      localJointOffset=<-4 0 0>
     }]
    simSpaceSettings {maxAngularAcceleration=4000, maxAngularVelocity=8, simSpaceAngularAlpha=1}
   }
   n18 = Anim Dynamics - Bone: dyn_L_chain_side_5(ComponentPose=@n17, Alpha=1, ExternalForce="0,0,0") {
    angularDampingOverride=1
    angularSpringConstant=10000
    bAngularSpring=true
    bOverrideAngularDamping=true
    bUseGravityOverride=true
    simulationSpace=RootRelative
    alphaScaleBiasClamp {bias=1.2, scale=-1}
    boundBone {boneName=dyn_L_chain_side_5}
    gravityOverride {z=-980}
    physicsBodyDefinitions [{
      boundBone {boneName=dyn_L_chain_side_5}
      boxExtents=<25 25 25>
      constraintSetup {angularLimitsMax=<0 2 20>, angularLimitsMin=<0 -2 0>, angularTarget=<0 0 0>}
      localJointOffset=<-4 0 0>
     }]
    simSpaceSettings {maxAngularAcceleration=4000, maxAngularVelocity=8, simSpaceAngularAlpha=1}
   }
   n19 = Anim Dynamics - Bone: dyn_R_chain_side_1(ComponentPose=@n18, Alpha=1, ExternalForce="0,0,0") {
    angularDampingOverride=1
    angularSpringConstant=10000
    bAngularSpring=true
    bOverrideAngularDamping=true
    bUseGravityOverride=true
    simulationSpace=RootRelative
    alphaScaleBiasClamp {bias=1.2, scale=-1}
    boundBone {boneName=dyn_R_chain_side_1}
    gravityOverride {z=-980}
    physicsBodyDefinitions [{
      boundBone {boneName=dyn_R_chain_side_1}
      boxExtents=<25 25 25>
      constraintSetup {angularLimitsMax=<0 2 20>, angularLimitsMin=<0 -2 0>, angularTarget=<0 0 0>}
      localJointOffset=<4 0 0>
     }]
    simSpaceSettings {maxAngularAcceleration=4000, maxAngularVelocity=8, simSpaceAngularAlpha=1}
   }
   n20 = Anim Dynamics - Bone: dyn_R_chain_side_2(ComponentPose=@n19, Alpha=1, ExternalForce="0,0,0") {
    angularDampingOverride=1
    angularSpringConstant=10000
    bAngularSpring=true
    bOverrideAngularDamping=true
    bUseGravityOverride=true
    simulationSpace=RootRelative
    alphaScaleBiasClamp {bias=1.2, scale=-1}
    boundBone {boneName=dyn_R_chain_side_2}
    gravityOverride {z=-980}
    physicsBodyDefinitions [{
      boundBone {boneName=dyn_R_chain_side_2}
      boxExtents=<25 25 25>
      constraintSetup {angularLimitsMax=<0 2 20>, angularLimitsMin=<0 -2 0>, angularTarget=<0 0 0>}
      localJointOffset=<4 0 0>
     }]
    simSpaceSettings {maxAngularAcceleration=4000, maxAngularVelocity=8, simSpaceAngularAlpha=1}
   }
   n21 = Anim Dynamics - Bone: dyn_R_chain_side_3(ComponentPose=@n20, Alpha=1, ExternalForce="0,0,0") {
    angularDampingOverride=1
    angularSpringConstant=10000
    bAngularSpring=true
    bOverrideAngularDamping=true
    bUseGravityOverride=true
    simulationSpace=RootRelative
    alphaScaleBiasClamp {bias=1.2, scale=-1}
    boundBone {boneName=dyn_R_chain_side_3}
    gravityOverride {z=-980}
    physicsBodyDefinitions [{
      boundBone {boneName=dyn_R_chain_side_3}
      boxExtents=<25 25 25>
      constraintSetup {angularLimitsMax=<0 2 20>, angularLimitsMin=<0 -2 0>, angularTarget=<0 0 0>}
      localJointOffset=<4 0 0>
     }]
    simSpaceSettings {maxAngularAcceleration=4000, maxAngularVelocity=8, simSpaceAngularAlpha=1}
   }
   n22 = Anim Dynamics - Bone: dyn_R_chain_side_4(ComponentPose=@n21, Alpha=1, ExternalForce="0,0,0") {
    angularDampingOverride=1
    angularSpringConstant=10000
    bAngularSpring=true
    bOverrideAngularDamping=true
    bUseGravityOverride=true
    simulationSpace=RootRelative
    alphaScaleBiasClamp {bias=1.2, scale=-1}
    boundBone {boneName=dyn_R_chain_side_4}
    gravityOverride {z=-980}
    physicsBodyDefinitions [{
      boundBone {boneName=dyn_R_chain_side_4}
      boxExtents=<25 25 25>
      constraintSetup {angularLimitsMax=<0 2 20>, angularLimitsMin=<0 -2 0>, angularTarget=<0 0 0>}
      localJointOffset=<4 0 0>
     }]
    simSpaceSettings {maxAngularAcceleration=4000, maxAngularVelocity=8, simSpaceAngularAlpha=1}
   }
   n23 = Anim Dynamics - Bone: dyn_R_chain_side_5(ComponentPose=@n22, Alpha=1, ExternalForce="0,0,0") {
    angularDampingOverride=1
    angularSpringConstant=10000
    bAngularSpring=true
    bOverrideAngularDamping=true
    bUseGravityOverride=true
    simulationSpace=RootRelative
    alphaScaleBiasClamp {bias=1.2, scale=-1}
    boundBone {boneName=dyn_R_chain_side_5}
    gravityOverride {z=-980}
    physicsBodyDefinitions [{
      boundBone {boneName=dyn_R_chain_side_5}
      boxExtents=<25 25 25>
      constraintSetup {angularLimitsMax=<0 2 20>, angularLimitsMin=<0 -2 0>, angularTarget=<0 0 0>}
      localJointOffset=<4 0 0>
     }]
    simSpaceSettings {maxAngularAcceleration=4000, maxAngularVelocity=8, simSpaceAngularAlpha=1}
   }
   n24 = Anim Dynamics - Bone: dyn_L_chain_back_1(ComponentPose=@n23, Alpha=1, ExternalForce="0,0,0") {
    angularDampingOverride=1
    angularSpringConstant=10000
    bAngularSpring=true
    bOverrideAngularDamping=true
    bUseGravityOverride=true
    simulationSpace=RootRelative
    alphaScaleBiasClamp {bias=1.2, scale=-1}
    boundBone {boneName=dyn_L_chain_back_1}
    gravityOverride {z=-980}
    physicsBodyDefinitions [{
      boundBone {boneName=dyn_L_chain_back_1}
      boxExtents=<25 25 25>
      constraintSetup {angularLimitsMax=<0 2 20>, angularLimitsMin=<0 -2 0>, angularTarget=<0 0 0>}
      localJointOffset=<-4 0 0>
     }]
    simSpaceSettings {maxAngularAcceleration=4000, maxAngularVelocity=8, simSpaceAngularAlpha=1}
   }
   n25 = Anim Dynamics - Bone: dyn_L_chain_back_2(ComponentPose=@n24, Alpha=1, ExternalForce="0,0,0") {
    angularDampingOverride=1
    angularSpringConstant=10000
    bAngularSpring=true
    bOverrideAngularDamping=true
    bUseGravityOverride=true
    simulationSpace=RootRelative
    alphaScaleBiasClamp {bias=1.2, scale=-1}
    boundBone {boneName=dyn_L_chain_back_2}
    gravityOverride {z=-980}
    physicsBodyDefinitions [{
      boundBone {boneName=dyn_L_chain_back_2}
      boxExtents=<25 25 25>
      constraintSetup {angularLimitsMax=<0 2 20>, angularLimitsMin=<0 -2 0>, angularTarget=<0 0 0>}
      localJointOffset=<-4 0 0>
     }]
    simSpaceSettings {maxAngularAcceleration=4000, maxAngularVelocity=8, simSpaceAngularAlpha=1}
   }
   n26 = Anim Dynamics - Bone: dyn_L_chain_back_3(ComponentPose=@n25, Alpha=1, ExternalForce="0,0,0") {
    angularDampingOverride=1
    angularSpringConstant=10000
    bAngularSpring=true
    bOverrideAngularDamping=true
    bUseGravityOverride=true
    simulationSpace=RootRelative
    alphaScaleBiasClamp {bias=1.2, scale=-1}
    boundBone {boneName=dyn_L_chain_back_3}
    gravityOverride {z=-980}
    physicsBodyDefinitions [{
      boundBone {boneName=dyn_L_chain_back_3}
      boxExtents=<25 25 25>
      constraintSetup {angularLimitsMax=<0 2 20>, angularLimitsMin=<0 -2 0>, angularTarget=<0 0 0>}
      localJointOffset=<-4 0 0>
     }]
    simSpaceSettings {maxAngularAcceleration=4000, maxAngularVelocity=8, simSpaceAngularAlpha=1}
   }
   n27 = Anim Dynamics - Bone: dyn_L_chain_back_4(ComponentPose=@n26, Alpha=1, ExternalForce="0,0,0") {
    angularDampingOverride=1
    angularSpringConstant=10000
    bAngularSpring=true
    bOverrideAngularDamping=true
    bUseGravityOverride=true
    simulationSpace=RootRelative
    alphaScaleBiasClamp {bias=1.2, scale=-1}
    boundBone {boneName=dyn_L_chain_back_4}
    gravityOverride {z=-980}
    physicsBodyDefinitions [{
      boundBone {boneName=dyn_L_chain_back_4}
      boxExtents=<25 25 25>
      constraintSetup {angularLimitsMax=<0 2 20>, angularLimitsMin=<0 -2 0>, angularTarget=<0 0 0>}
      localJointOffset=<-4 0 0>
     }]
    simSpaceSettings {maxAngularAcceleration=4000, maxAngularVelocity=8, simSpaceAngularAlpha=1}
   }
   n28 = Anim Dynamics - Bone: dyn_L_chain_back_5(ComponentPose=@n27, Alpha=1, ExternalForce="0,0,0") {
    angularDampingOverride=1
    angularSpringConstant=10000
    bAngularSpring=true
    bOverrideAngularDamping=true
    bUseGravityOverride=true
    simulationSpace=RootRelative
    alphaScaleBiasClamp {bias=1.2, scale=-1}
    boundBone {boneName=dyn_L_chain_back_5}
    gravityOverride {z=-980}
    physicsBodyDefinitions [{
      boundBone {boneName=dyn_L_chain_back_5}
      boxExtents=<25 25 25>
      constraintSetup {angularLimitsMax=<0 2 20>, angularLimitsMin=<0 -2 0>, angularTarget=<0 0 0>}
      localJointOffset=<-4 0 0>
     }]
    simSpaceSettings {maxAngularAcceleration=4000, maxAngularVelocity=8, simSpaceAngularAlpha=1}
   }
   n29 = Anim Dynamics - Bone: dyn_R_chain_back_1(ComponentPose=@n28, Alpha=1, ExternalForce="0,0,0") {
    angularDampingOverride=1
    angularSpringConstant=10000
    bAngularSpring=true
    bOverrideAngularDamping=true
    bUseGravityOverride=true
    simulationSpace=RootRelative
    alphaScaleBiasClamp {bias=1.2, scale=-1}
    boundBone {boneName=dyn_R_chain_back_1}
    gravityOverride {z=-980}
    physicsBodyDefinitions [{
      boundBone {boneName=dyn_R_chain_back_1}
      boxExtents=<25 25 25>
      constraintSetup {angularLimitsMax=<0 2 20>, angularLimitsMin=<0 -2 0>, angularTarget=<0 0 0>}
      localJointOffset=<4 0 0>
     }]
    simSpaceSettings {maxAngularAcceleration=4000, maxAngularVelocity=8, simSpaceAngularAlpha=1}
   }
   n30 = Anim Dynamics - Bone: dyn_R_chain_back_2(ComponentPose=@n29, Alpha=1, ExternalForce="0,0,0") {
    angularDampingOverride=1
    angularSpringConstant=10000
    bAngularSpring=true
    bOverrideAngularDamping=true
    bUseGravityOverride=true
    simulationSpace=RootRelative
    alphaScaleBiasClamp {bias=1.2, scale=-1}
    boundBone {boneName=dyn_R_chain_back_2}
    gravityOverride {z=-980}
    physicsBodyDefinitions [{
      boundBone {boneName=dyn_R_chain_back_2}
      boxExtents=<25 25 25>
      constraintSetup {angularLimitsMax=<0 2 20>, angularLimitsMin=<0 -2 0>, angularTarget=<0 0 0>}
      localJointOffset=<4 0 0>
     }]
    simSpaceSettings {maxAngularAcceleration=4000, maxAngularVelocity=8, simSpaceAngularAlpha=1}
   }
   n31 = Anim Dynamics - Bone: dyn_R_chain_back_3(ComponentPose=@n30, Alpha=1, ExternalForce="0,0,0") {
    angularDampingOverride=1
    angularSpringConstant=10000
    bAngularSpring=true
    bOverrideAngularDamping=true
    bUseGravityOverride=true
    simulationSpace=RootRelative
    alphaScaleBiasClamp {bias=1.2, scale=-1}
    boundBone {boneName=dyn_R_chain_back_3}
    gravityOverride {z=-980}
    physicsBodyDefinitions [{
      boundBone {boneName=dyn_R_chain_back_3}
      boxExtents=<25 25 25>
      constraintSetup {angularLimitsMax=<0 2 20>, angularLimitsMin=<0 -2 0>, angularTarget=<0 0 0>}
      localJointOffset=<4 0 0>
     }]
    simSpaceSettings {maxAngularAcceleration=4000, maxAngularVelocity=8, simSpaceAngularAlpha=1}
   }
   n32 = Anim Dynamics - Bone: dyn_R_chain_back_4(ComponentPose=@n31, Alpha=1, ExternalForce="0,0,0") {
    angularDampingOverride=1
    angularSpringConstant=10000
    bAngularSpring=true
    bOverrideAngularDamping=true
    bUseGravityOverride=true
    simulationSpace=RootRelative
    alphaScaleBiasClamp {bias=1.2, scale=-1}
    boundBone {boneName=dyn_R_chain_back_4}
    gravityOverride {z=-980}
    physicsBodyDefinitions [{
      boundBone {boneName=dyn_R_chain_back_4}
      boxExtents=<25 25 25>
      constraintSetup {angularLimitsMax=<0 2 20>, angularLimitsMin=<0 -2 0>, angularTarget=<0 0 0>}
      localJointOffset=<4 0 0>
     }]
    simSpaceSettings {maxAngularAcceleration=4000, maxAngularVelocity=8, simSpaceAngularAlpha=1}
   }
   n33 = Anim Dynamics - Bone: dyn_R_chain_back_5(ComponentPose=@n32, Alpha=1, ExternalForce="0,0,0") {
    angularDampingOverride=1
    angularSpringConstant=10000
    bAngularSpring=true
    bOverrideAngularDamping=true
    bUseGravityOverride=true
    simulationSpace=RootRelative
    alphaScaleBiasClamp {bias=1.2, scale=-1}
    boundBone {boneName=dyn_R_chain_back_5}
    gravityOverride {z=-980}
    physicsBodyDefinitions [{
      boundBone {boneName=dyn_R_chain_back_5}
      boxExtents=<25 25 25>
      constraintSetup {angularLimitsMax=<0 2 20>, angularLimitsMin=<0 -2 0>, angularTarget=<0 0 0>}
      localJointOffset=<4 0 0>
     }]
    simSpaceSettings {maxAngularAcceleration=4000, maxAngularVelocity=8, simSpaceAngularAlpha=1}
   }
   n34 = Anim Dynamics - Bone: dyn_C_chain_back_1(ComponentPose=@n33, Alpha=1, ExternalForce="0,0,0") {
    angularDampingOverride=1
    angularSpringConstant=10000
    bAngularSpring=true
    bOverrideAngularDamping=true
    bUseGravityOverride=true
    simulationSpace=RootRelative
    alphaScaleBiasClamp {bias=1.2, scale=-1}
    boundBone {boneName=dyn_C_chain_back_1}
    gravityOverride {z=-980}
    physicsBodyDefinitions [{
      boundBone {boneName=dyn_C_chain_back_1}
      boxExtents=<25 25 25>
      constraintSetup {angularLimitsMax=<0 2 20>, angularLimitsMin=<0 -2 0>, angularTarget=<0 0 0>}
      localJointOffset=<-4 0 0>
     }]
    simSpaceSettings {maxAngularAcceleration=4000, maxAngularVelocity=8, simSpaceAngularAlpha=1}
   }
   n35 = Anim Dynamics - Bone: dyn_C_chain_back_2(ComponentPose=@n34, Alpha=1, ExternalForce="0,0,0") {
    angularDampingOverride=1
    angularSpringConstant=10000
    bAngularSpring=true
    bOverrideAngularDamping=true
    bUseGravityOverride=true
    simulationSpace=RootRelative
    alphaScaleBiasClamp {bias=1.2, scale=-1}
    boundBone {boneName=dyn_C_chain_back_2}
    gravityOverride {z=-980}
    physicsBodyDefinitions [{
      boundBone {boneName=dyn_C_chain_back_2}
      boxExtents=<25 25 25>
      constraintSetup {angularLimitsMax=<0 2 20>, angularLimitsMin=<0 -2 0>, angularTarget=<0 0 0>}
      localJointOffset=<-4 0 0>
     }]
    simSpaceSettings {maxAngularAcceleration=4000, maxAngularVelocity=8, simSpaceAngularAlpha=1}
   }
   n36 = Anim Dynamics - Bone: dyn_C_chain_back_3(ComponentPose=@n35, Alpha=1, ExternalForce="0,0,0") {
    angularDampingOverride=1
    angularSpringConstant=10000
    bAngularSpring=true
    bOverrideAngularDamping=true
    bUseGravityOverride=true
    simulationSpace=RootRelative
    alphaScaleBiasClamp {bias=1.2, scale=-1}
    boundBone {boneName=dyn_C_chain_back_3}
    gravityOverride {z=-980}
    physicsBodyDefinitions [{
      boundBone {boneName=dyn_C_chain_back_3}
      boxExtents=<25 25 25>
      constraintSetup {angularLimitsMax=<0 2 20>, angularLimitsMin=<0 -2 0>, angularTarget=<0 0 0>}
      localJointOffset=<-4 0 0>
     }]
    simSpaceSettings {maxAngularAcceleration=4000, maxAngularVelocity=8, simSpaceAngularAlpha=1}
   }
   n37 = Anim Dynamics - Bone: dyn_C_chain_back_4(ComponentPose=@n36, Alpha=1, ExternalForce="0,0,0") {
    angularDampingOverride=1
    angularSpringConstant=10000
    bAngularSpring=true
    bOverrideAngularDamping=true
    bUseGravityOverride=true
    simulationSpace=RootRelative
    alphaScaleBiasClamp {bias=1.2, scale=-1}
    boundBone {boneName=dyn_C_chain_back_4}
    gravityOverride {z=-980}
    physicsBodyDefinitions [{
      boundBone {boneName=dyn_C_chain_back_4}
      boxExtents=<25 25 25>
      constraintSetup {angularLimitsMax=<0 2 20>, angularLimitsMin=<0 -2 0>, angularTarget=<0 0 0>}
      localJointOffset=<-4 0 0>
     }]
    simSpaceSettings {maxAngularAcceleration=4000, maxAngularVelocity=8, simSpaceAngularAlpha=1}
   }
   Anim Dynamics - Bone: dyn_C_chain_back_5(ComponentPose=@n37, Alpha=1, ExternalForce="0,0,0") {
    angularDampingOverride=1
    angularSpringConstant=10000
    bAngularSpring=true
    bOverrideAngularDamping=true
    bUseGravityOverride=true
    simulationSpace=RootRelative
    alphaScaleBiasClamp {bias=1.2, scale=-1}
    boundBone {boneName=dyn_C_chain_back_5}
    gravityOverride {z=-980}
    physicsBodyDefinitions [{
      boundBone {boneName=dyn_C_chain_back_5}
      boxExtents=<25 25 25>
      constraintSetup {angularLimitsMax=<0 2 20>, angularLimitsMin=<0 -2 0>, angularTarget=<0 0 0>}
      localJointOffset=<-4 0 0>
     }]
    simSpaceSettings {maxAngularAcceleration=4000, maxAngularVelocity=8, simSpaceAngularAlpha=1}
   }
   n38 = "Transform (Modify) Bone - Bone: dyn_R_chain_front_A_1"!(ComponentPose, bAlphaBoolEnabled:bool=True) -> (Pose) {
    alphaInputType=Bool
    rotationMode=BMM_Additive
    rotationSpace=BCS_BoneSpace
    alphaBoolBlend {blendInTime=0.1, blendOutTime=0.1}
    boneToModify {boneName=dyn_R_chain_front_A_1}
    rotation {pitch=40, yaw=-30}
   }
   n39 = "Transform (Modify) Bone - Bone: dyn_R_chain_front_B_1"!(ComponentPose=@n38, bAlphaBoolEnabled:bool=True) -> (Pose) {
    alphaInputType=Bool
    rotationMode=BMM_Additive
    rotationSpace=BCS_BoneSpace
    alphaBoolBlend {blendInTime=0.1, blendOutTime=0.1}
    boneToModify {boneName=dyn_R_chain_front_B_1}
    rotation {pitch=-20, yaw=-50}
   }
   "Transform (Modify) Bone - Bone: dyn_R_chain_side_3"!(ComponentPose=@n39, bAlphaBoolEnabled:bool=True) -> (Pose) {
    alphaInputType=Bool
    rotationMode=BMM_Additive
    rotationSpace=BCS_BoneSpace
    alphaBoolBlend {blendInTime=0.1, blendOutTime=0.1}
    boneToModify {boneName=dyn_R_chain_side_3}
    rotation {pitch=-70, yaw=-25}
   }
   n40 = RigidBody!(ComponentPose=Local To Component(LocalPose=Use cached pose 'startPose'()), Alpha=0.6, OverrideWorldGravity:Vector="0,0,-980", ExternalForce) -> (Pose) {
    bClampLinearTranslationLimitToRefPose=true
    bOverrideWorldGravity=true
    overridePhysicsAsset=&/F_MED_LadderDoor_Skydive_Physics
    simulationSpace=BaseBoneSpace
    alphaScaleBiasClamp {interpSpeedDecreasing=20, interpSpeedIncreasing=20}
    baseBoneRef {boneName=pelvis}
    componentAppliedLinearAccClamp=<0 0 0>
    overrideWorldGravity {z=-980}
    simSpaceSettings {
     dampingAlpha=0
     maxAngularAcceleration=15
     maxAngularVelocity=5
     maxLinearAcceleration=400
     maxLinearVelocity=200
     velocityScaleZ=0
     worldAlpha=1
    }
   }
   n41 = "Transform (Modify) Bone - Bone: dyn_L_chain_front_4"!(ComponentPose=@n40, bAlphaBoolEnabled:bool=True) -> (Pose) {
    alphaInputType=Bool
    rotationMode=BMM_Additive
    rotationSpace=BCS_BoneSpace
    alphaBoolBlend {blendInTime=0.7, blendOutTime=0.7}
    boneToModify {boneName=dyn_L_chain_front_4}
    rotation {yaw=-40}
   }
   n42 = "Transform (Modify) Bone - Bone: dyn_R_chain_front_A_2"!(ComponentPose=@n41, bAlphaBoolEnabled:bool=True) -> (Pose) {
    alphaInputType=Bool
    rotationMode=BMM_Additive
    rotationSpace=BCS_BoneSpace
    alphaBoolBlend {blendInTime=0.7, blendOutTime=0.7}
    boneToModify {boneName=dyn_R_chain_front_A_2}
    rotation {yaw=-40}
   }
   n43 = "Transform (Modify) Bone - Bone: dyn_R_chain_front_B_2"!(ComponentPose=@n42, bAlphaBoolEnabled:bool=True) -> (Pose) {
    alphaInputType=Bool
    rotationMode=BMM_Additive
    rotationSpace=BCS_BoneSpace
    alphaBoolBlend {blendInTime=0.7, blendOutTime=0.7}
    boneToModify {boneName=dyn_R_chain_front_B_2}
    rotation {yaw=-40}
   }
   n44 = Component To Local(ComponentPose=@n43)
   n45 = RigidBody(ComponentPose=Local To Component(LocalPose=Use cached pose 'startPose'()), OverrideWorldGravity="0,0,-980") {
    bClampLinearTranslationLimitToRefPose=true
    bOverrideWorldGravity=true
    overridePhysicsAsset=&/F_MED_LadderDoor_Physics
    simulationSpace=BaseBoneSpace
    alphaScaleBiasClamp {interpSpeedDecreasing=20, interpSpeedIncreasing=20}
    baseBoneRef {boneName=root}
    componentAppliedLinearAccClamp=<3500 3500 3500>
    componentLinearAccScale=<0.2 0.2 0.2>
    componentLinearVelScale=<0.2 0.2 0.2>
    overrideWorldGravity {z=-980}
    simSpaceSettings {
     dampingAlpha=0
     maxAngularAcceleration=50
     maxAngularVelocity=5
     maxLinearAcceleration=2500
     maxLinearVelocity=800
     velocityScaleZ=0.35
     worldAlpha=1
     externalLinearDragV=<1 1 0.5>
    }
   }
   n46 = Component To Local(ComponentPose=@n45)
   n47 = Get bIsSkyDiving()
   n48 = "Transform (Modify) Bone - Bone: dyn_L_chain_front_master"!(
    ComponentPose=Local To Component(LocalPose=Blend Poses by bool(BlendPose_0=@n44, BlendPose_1=@n46, bActiveValue=@n47))
    Alpha=1
    Scale="1,1,1"
   ) -> (Pose) {
    rotationMode=BMM_Additive
    rotationSpace=BCS_BoneSpace
    translationMode=BMM_Additive
    translationSpace=BCS_BoneSpace
    alphaScaleBiasClamp {bias=1, scale=-1}
    boneToModify {boneName=dyn_L_chain_front_master}
    rotation {yaw=6}
    translation {y=-0.4}
   }
   n49 = "Transform (Modify) Bone - Bone: dyn_L_chain_side_master"!(ComponentPose=@n48, Alpha=1, Scale="1,1,1") -> (Pose) {
    rotationMode=BMM_Additive
    rotationSpace=BCS_BoneSpace
    translationMode=BMM_Additive
    translationSpace=BCS_BoneSpace
    alphaScaleBiasClamp {bias=1, scale=-1}
    boneToModify {boneName=dyn_L_chain_side_master}
    rotation {yaw=6}
    translation {y=-0.3}
   }
   n50 = "Transform (Modify) Bone - Bone: dyn_L_chain_back_master"!(ComponentPose=@n49, Alpha=1, Scale="1,1,1") -> (Pose) {
    rotationMode=BMM_Additive
    rotationSpace=BCS_BoneSpace
    translationMode=BMM_Additive
    translationSpace=BCS_BoneSpace
    alphaScaleBiasClamp {bias=1, scale=-1}
    boneToModify {boneName=dyn_L_chain_back_master}
    rotation {yaw=6}
    translation {y=-1.5}
   }
   n51 = "Transform (Modify) Bone - Bone: dyn_R_chain_front_master"!(ComponentPose=@n50, Alpha=1, Scale="1,1,1") -> (Pose) {
    rotationMode=BMM_Additive
    rotationSpace=BCS_BoneSpace
    translationMode=BMM_Additive
    translationSpace=BCS_BoneSpace
    alphaScaleBiasClamp {bias=1, scale=-1}
    boneToModify {boneName=dyn_R_chain_front_master}
    rotation {yaw=6}
    translation {y=0.4}
   }
   n52 = "Transform (Modify) Bone - Bone: dyn_R_chain_side_master"!(ComponentPose=@n51, Alpha=1, Scale="1,1,1") -> (Pose) {
    rotationMode=BMM_Additive
    rotationSpace=BCS_BoneSpace
    translationMode=BMM_Additive
    translationSpace=BCS_BoneSpace
    alphaScaleBiasClamp {bias=1, scale=-1}
    boneToModify {boneName=dyn_R_chain_side_master}
    rotation {yaw=6}
    translation {y=0.3}
   }
   n53 = "Transform (Modify) Bone - Bone: dyn_R_chain_back_master"!(ComponentPose=@n52, Alpha=1, Scale="1,1,1") -> (Pose) {
    rotationMode=BMM_Additive
    rotationSpace=BCS_BoneSpace
    translationMode=BMM_Additive
    translationSpace=BCS_BoneSpace
    alphaScaleBiasClamp {bias=1, scale=-1}
    boneToModify {boneName=dyn_R_chain_back_master}
    rotation {yaw=6}
    translation {y=1.5}
   }
   n54 = "Transform (Modify) Bone - Bone: dyn_C_chain_back_master"!(ComponentPose=@n53, Alpha=1, Scale="1,1,1") -> (Pose) {
    rotationMode=BMM_Additive
    rotationSpace=BCS_BoneSpace
    translationMode=BMM_Additive
    translationSpace=BCS_BoneSpace
    alphaScaleBiasClamp {bias=1, scale=-1}
    boneToModify {boneName=dyn_C_chain_back_master}
    rotation {yaw=6}
    translation {y=-1.5}
   }
   n55 = "Transform (Modify) Bone - Bone: dyn_L_chain_front_4"(ComponentPose=@n54, bAlphaBoolEnabled=True) {
    alphaInputType=Bool
    scaleMode=BMM_Additive
    scaleSpace=BCS_BoneSpace
    alphaBoolBlend {blendInTime=0.7, blendOutTime=0.7}
    boneToModify {boneName=dyn_L_chain_front_4}
    scale {x=0.8}
   }
   n56 = "Transform (Modify) Bone - Bone: dyn_R_chain_front_A_2"(ComponentPose=@n55, bAlphaBoolEnabled=True) {
    alphaInputType=Bool
    scaleMode=BMM_Additive
    scaleSpace=BCS_BoneSpace
    alphaBoolBlend {blendInTime=0.7, blendOutTime=0.7}
    boneToModify {boneName=dyn_R_chain_front_A_2}
    scale {x=0.8}
   }
   n57 = "Transform (Modify) Bone - Bone: dyn_R_chain_front_B_2"(ComponentPose=@n56, bAlphaBoolEnabled=True) {
    alphaInputType=Bool
    scaleMode=BMM_Additive
    scaleSpace=BCS_BoneSpace
    alphaBoolBlend {blendInTime=0.7, blendOutTime=0.7}
    boneToModify {boneName=dyn_R_chain_front_B_2}
    scale {x=0.8}
   }
   n58 = "Transform (Modify) Bone - Bone: dyn_L_chain_side_master"!(ComponentPose=@n57, bAlphaBoolEnabled:bool=True) -> (Pose) {
    alphaInputType=Bool
    scaleMode=BMM_Additive
    scaleSpace=BCS_BoneSpace
    alphaBoolBlend {blendInTime=0.7, blendOutTime=0.7}
    boneToModify {boneName=dyn_L_chain_side_master}
    scale {x=0.8}
   }
   n59 = "Transform (Modify) Bone - Bone: dyn_R_chain_side_master"!(ComponentPose=@n58, bAlphaBoolEnabled:bool=True) -> (Pose) {
    alphaInputType=Bool
    scaleMode=BMM_Additive
    scaleSpace=BCS_BoneSpace
    alphaBoolBlend {blendInTime=0.7, blendOutTime=0.7}
    boneToModify {boneName=dyn_R_chain_side_master}
    scale {x=0.8}
   }
   n60 = "Transform (Modify) Bone - Bone: dyn_L_chain_back_master"!(ComponentPose=@n59, bAlphaBoolEnabled:bool=True) -> (Pose) {
    alphaInputType=Bool
    scaleMode=BMM_Additive
    scaleSpace=BCS_BoneSpace
    alphaBoolBlend {blendInTime=0.7, blendOutTime=0.7}
    boneToModify {boneName=dyn_L_chain_back_master}
    scale {x=0.7}
   }
   n61 = "Transform (Modify) Bone - Bone: dyn_C_chain_back_master"!(ComponentPose=@n60, bAlphaBoolEnabled:bool=True) -> (Pose) {
    alphaInputType=Bool
    scaleMode=BMM_Additive
    scaleSpace=BCS_BoneSpace
    alphaBoolBlend {blendInTime=0.7, blendOutTime=0.7}
    boneToModify {boneName=dyn_C_chain_back_master}
    scale {x=0.7}
   }
   n62 = "Transform (Modify) Bone - Bone: dyn_R_chain_back_master"!(ComponentPose=@n61, bAlphaBoolEnabled:bool=True) -> (Pose) {
    alphaInputType=Bool
    scaleMode=BMM_Additive
    scaleSpace=BCS_BoneSpace
    alphaBoolBlend {blendInTime=0.7, blendOutTime=0.7}
    boneToModify {boneName=dyn_R_chain_back_master}
    scale {x=0.7}
   }
   n63 = "Transform (Modify) Bone - Bone: dyn_L_chain_front_3"!(ComponentPose=@n62, bAlphaBoolEnabled:bool=True) -> (Pose) {
    alphaInputType=Bool
    rotationMode=BMM_Additive
    rotationSpace=BCS_BoneSpace
    alphaBoolBlend {blendInTime=0.7, blendOutTime=0.7}
    boneToModify {boneName=dyn_L_chain_front_3}
    rotation {yaw=-50}
   }
   n64 = "Transform (Modify) Bone - Bone: dyn_R_chain_front_A_1"(ComponentPose=@n63, bAlphaBoolEnabled=True) {
    alphaInputType=Bool
    rotationMode=BMM_Additive
    rotationSpace=BCS_BoneSpace
    alphaBoolBlend {blendInTime=0.7, blendOutTime=0.7}
    boneToModify {boneName=dyn_R_chain_front_A_1}
    rotation {yaw=-50}
   }
   n65 = "Transform (Modify) Bone - Bone: dyn_R_chain_front_B_1"(ComponentPose=@n64, bAlphaBoolEnabled=True) {
    alphaInputType=Bool
    rotationMode=BMM_Additive
    rotationSpace=BCS_BoneSpace
    alphaBoolBlend {blendInTime=0.7, blendOutTime=0.7}
    boneToModify {boneName=dyn_R_chain_front_B_1}
    rotation {yaw=-50}
   }
   n66 = RigidBody(ComponentPose=Local To Component(LocalPose=Use cached pose 'startPose'()), OverrideWorldGravity="0,0,-980") {
    bClampLinearTranslationLimitToRefPose=true
    bOverrideWorldGravity=true
    overridePhysicsAsset=&/F_MED_LadderDoor_Idle_Physics
    simulationSpace=BaseBoneSpace
    baseBoneRef {boneName=root}
    componentAppliedLinearAccClamp=<3500 3500 3500>
    componentLinearAccScale=<0.2 0.2 0.2>
    componentLinearVelScale=<0.2 0.2 0.2>
    overrideWorldGravity {z=-980}
    simSpaceSettings {
     dampingAlpha=0
     maxAngularAcceleration=50
     maxAngularVelocity=5
     maxLinearAcceleration=2500
     maxLinearVelocity=800
     velocityScaleZ=0.35
     worldAlpha=1
     externalLinearDragV=<1.6 1.6 1.8>
    }
   }
   Output Pose(Result=Blend Poses by bool(
     BlendPose_0=Component To Local(ComponentPose=@n66)
     BlendPose_1=Component To Local(ComponentPose=@n65)
     bActiveValue=Get CustomIdleIsPlayingBool() -> (CustomIdleIsPlayingBool:bool)
    )) {layerGroup=None}
   n67 = Get Sim Space Settings() -> (Sim Space Settings:Anim Phys Sim Space Settings Structure)
   n68 = Control Rig!(
    Source=Copy Pose From Mesh() {bUseAttachedParent=true}
    Alpha
    customIdlePlaying:bool=Get CustomIdleIsPlayingBool()
    isCrouchMotionless:bool=Get isCrouchMotionless() -> (isCrouchMotionless:bool)
    isSkydive:bool=Get bIsSkydiveFloating()
    isFrontendNotEmoting:double=Get isFrontendNotEmoting() -> (isFrontendNotEmoting:double)
   ) -> (Pose) {controlRigAssetReference {blueprintRigClass=&/F_MED_LadderDoor_CtrlRig.F_MED_LadderDoor_CtrlRig_C}}
   n69 = "Constraint - dyn_sash_master (PPPPPP)"!(
    ComponentPose=Local To Component(LocalPose=@n68)
    Alpha=1
    ConstraintWeights_0=0.0003
    ConstraintWeights_1:float=0.0003
    ConstraintWeights_2:float=0.1279
    ConstraintWeights_3:float=0.0034
    ConstraintWeights_4:float=0.8635
    ConstraintWeights_5:float=0.0046
   ) -> (Pose) {
    boneToModify {boneName=dyn_sash_master}
    constraintSetup [
     {targetBone {boneName=pelvis}}
     {targetBone {boneName=thigh_r}}
     {targetBone {boneName=dyn_R_chain_front_1}}
     {targetBone {boneName=dyn_R_chain_front_2}}
     {targetBone {boneName=dyn_R_chain_side_1}}
     {targetBone {boneName=dyn_R_chain_side_2}}
    ]
    constraintWeights [1,1,1,1,1,1]
   }
   n70 = "Transform (Modify) Bone - Bone: dyn_R_chest"!(ComponentPose=@n69, bAlphaBoolEnabled:bool=True, Translation="0,0,1", Rotation="0,0,0", Scale="1,1,1") -> (Pose) {alphaInputType=Bool, translationMode=BMM_Additive, translationSpace=BCS_BoneSpace, boneToModify {boneName=dyn_R_chest}, translation {z=1}}
   n71 = "Transform (Modify) Bone - Bone: dyn_L_chest"!(ComponentPose=@n70, bAlphaBoolEnabled:bool=True, Translation="0,0,1", Rotation="0,0,0", Scale="1,1,1") -> (Pose) {alphaInputType=Bool, translationMode=BMM_Additive, translationSpace=BCS_BoneSpace, boneToModify {boneName=dyn_L_chest}, translation {z=1}}
   n72 = Anim Dynamics - Bone: dyn_R_chest!(
    ComponentPose=@n71
    bAlphaBoolEnabled:bool=True
    AngularSpringConstant:float=5000
    SimSpaceSettings:Anim Phys Sim Space Settings Structure=@n67
   ) -> (Pose) {
    alphaInputType=Bool
    angularBiasOverride=1
    angularDampingOverride=0.9999
    angularSpringConstant=5000
    bAngularSpring=true
    bGravityOverrideInSimSpace=true
    bOverrideAngularBias=true
    bOverrideAngularDamping=true
    bUseGravityOverride=true
    linearDampingOverride=0.9999
    linearSpringConstant=100
    simulationSpace=RootRelative
    boundBone {boneName=dyn_R_chest}
    componentAppliedLinearAccClamp=<2 2 2>
    gravityOverride {z=-980}
    physicsBodyDefinitions [{
      boundBone {boneName=dyn_R_chest}
      boxExtents=<21 21 21>
      constraintSetup {angularLimitsMax=<0 1 7>, angularLimitsMin=<0 -7 -7>, angularTarget=<1 0 0>}
      localJointOffset=<-2 0 0>
     }]
    relativeSpaceBone {boneName=spine_04}
    simSpaceSettings {maxAngularAcceleration=45, maxAngularVelocity=8, simSpaceAngularAlpha=0.25}
   }
   n73 = Anim Dynamics - Bone: dyn_L_chest!(
    ComponentPose=@n72
    bAlphaBoolEnabled:bool=True
    AngularSpringConstant:float=5000
    SimSpaceSettings:Anim Phys Sim Space Settings Structure=@n67
   ) -> (Pose) {
    alphaInputType=Bool
    angularBiasOverride=1
    angularDampingOverride=0.9999
    angularSpringConstant=5000
    bAngularSpring=true
    bGravityOverrideInSimSpace=true
    bOverrideAngularBias=true
    bOverrideAngularDamping=true
    bUseGravityOverride=true
    linearDampingOverride=0.9999
    linearSpringConstant=100
    simulationSpace=RootRelative
    boundBone {boneName=dyn_L_chest}
    componentAppliedLinearAccClamp=<2 2 2>
    gravityOverride {z=-980}
    physicsBodyDefinitions [{
      boundBone {boneName=dyn_L_chest}
      boxExtents=<20 20 20>
      constraintSetup {angularLimitsMax=<0 1 7>, angularLimitsMin=<0 -7 -7>, angularTarget=<1 0 0>}
      localJointOffset=<2 0 0>
     }]
    relativeSpaceBone {boneName=spine_04}
    simSpaceSettings {maxAngularAcceleration=45, maxAngularVelocity=8, simSpaceAngularAlpha=0.25}
   }
   n74 = Anim Dynamics - Bone: dyn_R_chest!(ComponentPose=@n73, bAlphaBoolEnabled=True, GravityOverride:Vector="0,0,-200", LinearSpringConstant:float=180) -> (Pose) {
    alphaInputType=Bool
    angularBiasOverride=1
    angularDampingOverride=0.99
    angularSpringConstant=10000
    bGravityOverrideInSimSpace=true
    bLinearSpring=true
    bOverrideAngularBias=true
    bOverrideLinearDamping=true
    bUseGravityOverride=true
    linearDampingOverride=1
    linearSpringConstant=180
    simulationSpace=RootRelative
    boundBone {boneName=dyn_R_chest}
    componentAppliedLinearAccClamp=<5 5 5>
    gravityOverride {z=-200}
    physicsBodyDefinitions [{
      boundBone {boneName=dyn_R_chest}
      boxExtents=<80 80 80>
      constraintSetup {angularLimitsMax=<0 0 0>, angularLimitsMin=<0 0 0>, angularTarget=<1 0 0>, linearAxesMax=<0 0.5 1>, linearAxesMin=<0 -0.5 -0.5>}
      localJointOffset=<-0.2 0 0>
     }]
    relativeSpaceBone {boneName=spine_04}
    simSpaceSettings {maxAngularAcceleration=50, maxAngularVelocity=10, simSpaceAngularAlpha=0.5}
   }
   n75 = Anim Dynamics - Bone: dyn_L_chest!(ComponentPose=@n74, bAlphaBoolEnabled=True, GravityOverride:Vector="0,0,-200", LinearSpringConstant:float=180) -> (Pose) {
    alphaInputType=Bool
    angularBiasOverride=1
    angularDampingOverride=0.99
    angularSpringConstant=10000
    bGravityOverrideInSimSpace=true
    bLinearSpring=true
    bOverrideAngularBias=true
    bOverrideLinearDamping=true
    bUseGravityOverride=true
    linearDampingOverride=1
    linearSpringConstant=180
    simulationSpace=RootRelative
    boundBone {boneName=dyn_L_chest}
    componentAppliedLinearAccClamp=<5 5 5>
    gravityOverride {z=-200}
    physicsBodyDefinitions [{
      boundBone {boneName=dyn_L_chest}
      boxExtents=<90 90 90>
      constraintSetup {angularLimitsMax=<0 0 0>, angularLimitsMin=<0 0 0>, angularTarget=<1 0 0>, linearAxesMax=<0 0.5 1>, linearAxesMin=<0 -0.5 -0.5>}
      localJointOffset=<0.2 0 0>
     }]
    relativeSpaceBone {boneName=spine_04}
    simSpaceSettings {maxAngularAcceleration=50, maxAngularVelocity=10, simSpaceAngularAlpha=0.5}
   }
   Save cached pose 'startPose'(Pose=Component To Local(ComponentPose=@n75)) {cacheName=startPose}
  ]}
 EventGraph {nodes [
   Event BlueprintUpdateAnimation()
   > n1 = Sequence!(execute) -> (then_0, then_1, then_2:Exec, then_3:Exec, then_4:Exec, then_5:Exec)
   Branch(execute=@n1.then_1, Condition=Get bIsInFrontEnd())
   > n2 = Set CustomIdleIsPlayingFloat(CustomIdleIsPlayingFloat:float=GetCurveValue(self=GetOwnerFrontEndAnimInstance(), CurveName=Frontend_Idle_Bypass)) -> (Output_Get:float)
   > Set CustomIdleIsPlayingBool(CustomIdleIsPlayingBool:bool="Equal (Float)"(A=@n2, B=1)) -> (Output_Get:bool)
   n3 = Get bIsMoving2D()
   n4 = Get bIsAccelerating2D()
   n5 = Branch(execute=@n1.then_0, Condition=AND Boolean!(A=Get bIsCrouching(), B=NOT Boolean(A=@n4), C:bool=NOT Boolean(A=@n3)) -> (ReturnValue))
   > Set isCrouchMotionless(isCrouchMotionless:bool=true) -> (Output_Get:bool)
   Set isCrouchMotionless(execute=@n5.else, isCrouchMotionless=false)
   Set GravityOverride(execute=@n1.then_3, GravityOverride=SelectVector(A="0,0,-400", B="0,0,-2400", bPickA=Get bIsSkydiveDiveMode()))
   n6 = Branch(
    execute=@n1.then_4
    Condition="Not Equal (Boolean)"(A=Get previousRigidBodyEnabled() -> (previousRigidBodyEnabled:bool), B=Get bRigidBodyEnabled()) -> (ReturnValue=~)
   )
   ClearSkinWeightProfile!(
    execute=[@n9.then, @n10.then, @n6.then]
    self:Skinned Mesh Component Object Reference=Get BodySkelPart() -> (BodySkelPart:Skeletal Mesh Component Object Reference)
    InLayer:ESkinWeightProfileLayer Enum=Secondary
   ) -> (then)
   > Set previousRigidBodyEnabled(previousRigidBodyEnabled:bool=Get bRigidBodyEnabled()) -> (Output_Get:bool)
   n7 = NOT Boolean(A=Get bIsPlayingEmote())
   n8 = Get bIsInFrontEnd()
   Set isFrontendNotEmoting(execute=@n1.then_5, isFrontendNotEmoting:double="To Float (Boolean)"(InBool=AND Boolean!(A=@n8, B=@n7) -> (ReturnValue))) -> (Output_Get:double)
   Set skydiveFloatChangeAlpha(execute=@n1.then_2, skydiveFloatChangeAlpha:double=SelectFloat(A=0.5, B=1, bPickA=Get bIsSkydiveFloating())) -> (Output_Get:double)
   Event OnCharacterCustomizationLoaded()
   > Set BodySkelPart(BodySkelPart:Skeletal Mesh Component Object Reference=GetPartSkeletalMeshComponent()) -> (Output_Get:Skeletal Mesh Component Object Reference)
   n9 = Event OnEndTargeting() {is=title K2Node_Event}
   Event BlueprintInitializeAnimation()
   > Set previousRigidBodyEnabled(previousRigidBodyEnabled=false)
   n10 = Event OnEndCrouching()
   Event OnBeginCrouching()
   > n11 = Branch(Condition=Get bIsTargeting())
   SetSkinWeightProfile!(
    execute=[@n11.then, @n12.then]
    self:Skinned Mesh Component Object Reference=Get BodySkelPart()
    InProfileName:Name=Alternative
    InLayer:ESkinWeightProfileLayer Enum=Secondary
   ) -> (then, ReturnValue:bool)
   Event OnBeginTargeting() {is=title K2Node_Event}
   > n12 = Branch(Condition=Get bIsCrouching())
   n13 = Event OnParachuteClosed()
   n14 = Sequence!(execute=[@n13.then, @n15.then]) -> (then_0, then_1)
   > ResetDynamics(InTeleportType=TeleportPhysics)
   ResetDynamics(execute=@n14.then_1, InTeleportType=ResetPhysics)
   n15 = Event OnExitGrindRail() {is=title K2Node_Event}
  ]}
]
variables [
 BodySkelPart {type=SkeletalMeshComponent}
 CustomIdleIsPlayingBool {type=boolean}
 CustomIdleIsPlayingFloat {type=number}
 GravityOverride {type=Vector}
 isCrouchMotionless {type=boolean}
 isCrouchTargeting {type=boolean}
 isFrontendNotEmoting {type=number}
 previousRigidBodyEnabled {type=boolean}
 Sim Space Settings {
  type=AnimPhysSimSpaceSettings
  default {maxAngularAcceleration=10000, maxAngularVelocity=10000, simSpaceAngularAlpha=0, externalAngularVelocity=<0 0 0>}
 }
 skydiveFloatChangeAlpha {type=number}
]
