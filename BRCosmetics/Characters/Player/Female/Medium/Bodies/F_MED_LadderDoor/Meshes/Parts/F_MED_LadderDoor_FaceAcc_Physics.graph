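# Physics asset for the face accessory mesh of the LadderDoor female medium character, covering left and right five-bone hair chains.
# Capsules on the hair bones sit alongside capsules on upper spine, neck, clavicles and upper arms, which have no constraints and appear to serve as colliders.
# Each chain hangs from a master bone through swing and twist limits, tightest at the tip.
# Cross-constraints with free rotation link the two sides at the third and fifth bones.
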PhysicsAsset /BRCosmetics/Characters/Player/Female/Medium/Bodies/F_MED_LadderDoor/Meshes/Parts/F_MED_LadderDoor_FaceAcc_Physics
bodies {
 clavicle_l [{shapeType=Capsule, length=8, radius=7, shapeName=clavicle_l_capsule, center=<5 0 -2>, rotation=<rot 95 0 0>}]
 clavicle_r [{shapeType=Capsule, length=8, radius=7, shapeName=clavicle_r_capsule, center=<-5 0 2>, rotation=<rot 95 0 0>}]
 dyn_L_hair_1 [{shapeType=Capsule, length=8, radius=5, shapeName=dyn_L_hair_1_capsule, center=<5.0712 0 0>, rotation=<rot 90 0 0>}]
 dyn_L_hair_2 [{shapeType=Capsule, length=8, radius=5, shapeName=dyn_L_hair_2_capsule, center=<3.9314 0 0>, rotation=<rot 90 0 0>}]
 dyn_L_hair_3 [{shapeType=Capsule, length=8, radius=6, shapeName=dyn_L_hair_3_capsule, center=<3.5 0 1.5>, rotation=<rot 90 0 0>}]
 dyn_L_hair_4 [{shapeType=Capsule, length=8, radius=8, shapeName=dyn_L_hair_4_capsule, center=<6 2.5 3>, rotation=<rot 90 0 0>}]
 dyn_L_hair_5 [{shapeType=Capsule, length=8, radius=6.5, shapeName=dyn_L_hair_5_capsule, center=<2.3375 0 0>, rotation=<rot 90 0 0>}]
 dyn_L_hair_mstr [{shapeType=Capsule, length=0.505, radius=0.505, shapeName=dyn_L_hair_mstr_capsule}]
 dyn_R_hair_1 [{shapeType=Capsule, length=8, radius=5, shapeName=dyn_R_hair_1_capsule, center=<-5.0712 0 0>, rotation=<rot 90 0 0>}]
 dyn_R_hair_2 [{shapeType=Capsule, length=8, radius=5, shapeName=dyn_R_hair_2_capsule, center=<-3.9314 0 0>, rotation=<rot 90 0 0>}]
 dyn_R_hair_3 [{shapeType=Capsule, length=8, radius=6, shapeName=dyn_R_hair_3_capsule, center=<-3.5 0 -1.5>, rotation=<rot 90 0 0>}]
 dyn_R_hair_4 [{shapeType=Capsule, length=8, radius=8, shapeName=dyn_R_hair_4_capsule, center=<-6 -2.5 -3>, rotation=<rot 90 0 0>}]
 dyn_R_hair_5 [{shapeType=Capsule, length=8, radius=6.5, shapeName=dyn_R_hair_5_capsule, center=<-2.3375 0 0>, rotation=<rot 90 0 0>}]
 dyn_R_hair_mstr [{shapeType=Capsule, length=0.505, radius=0.505, shapeName=dyn_R_hair_mstr_capsule}]
 neck_01 [{shapeType=Capsule, length=8, radius=6, shapeName=neck_01_capsule, center=<0.7228 0.9361 0>, rotation=<rot 90 0 0>}]
 spine_04 [{shapeType=Capsule, length=11, radius=11, shapeName=spine_05_capsule, center=<0 2.0669 0>}]
 spine_05 [{shapeType=Capsule, length=11, radius=11, shapeName=spine_05_capsule, center=<0 0.8985 0>}]
 upperarm_l [{shapeType=Capsule, length=20, radius=6.5, shapeName=upperarm_l_capsule, center=<12 0 -1>, rotation=<rot 90 0 0>}]
 upperarm_r [{shapeType=Capsule, length=20, radius=6.5, shapeName=upperarm_l_capsule, center=<-12 0 1>, rotation=<rot 90 0 0>}]
}
constraints [
 {bone1Name=dyn_L_hair_1, bone2Name=dyn_L_hair_mstr, swing1LimitDegrees=20, swing2LimitDegrees=20, twistLimitDegrees=5}
 {bone1Name=dyn_L_hair_2, bone2Name=dyn_L_hair_1, swing1LimitDegrees=20, swing2LimitDegrees=20, twistLimitDegrees=5}
 {bone1Name=dyn_L_hair_3, bone2Name=dyn_L_hair_2, swing1LimitDegrees=30, swing2LimitDegrees=30, twistLimitDegrees=5}
 {bone1Name=dyn_L_hair_4, bone2Name=dyn_L_hair_3, swing1LimitDegrees=20, swing2LimitDegrees=10, twistLimitDegrees=5}
 {bone1Name=dyn_L_hair_5, bone2Name=dyn_L_hair_4, swing1LimitDegrees=10, swing2LimitDegrees=0, twistLimitDegrees=0}
 {bone1Name=dyn_R_hair_1, bone2Name=dyn_R_hair_mstr, swing1LimitDegrees=20, swing2LimitDegrees=20, twistLimitDegrees=5}
 {bone1Name=dyn_R_hair_2, bone2Name=dyn_R_hair_1, swing1LimitDegrees=20, swing2LimitDegrees=20, twistLimitDegrees=5}
 {bone1Name=dyn_R_hair_3, bone2Name=dyn_R_hair_2, swing1LimitDegrees=30, swing2LimitDegrees=30, twistLimitDegrees=5}
 {bone1Name=dyn_R_hair_4, bone2Name=dyn_R_hair_3, swing1LimitDegrees=20, swing2LimitDegrees=10, twistLimitDegrees=5}
 {bone1Name=dyn_R_hair_5, bone2Name=dyn_R_hair_4, swing1LimitDegrees=10, swing2LimitDegrees=0, twistLimitDegrees=0}
 {
  bone1Name=dyn_R_hair_5
  bone2Name=dyn_L_hair_5
  swing1LimitDegrees=45
  swing1Motion=Free
  swing2LimitDegrees=45
  swing2Motion=Free
  twistLimitDegrees=45
  twistMotion=Free
 }
 {
  bone1Name=dyn_R_hair_3
  bone2Name=dyn_L_hair_3
  swing1LimitDegrees=45
  swing1Motion=Free
  swing2LimitDegrees=45
  swing2Motion=Free
  twistLimitDegrees=45
  twistMotion=Free
 }
]
