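# Animation Blueprint for the F_MED_LadderDoor female face accessory, driving secondary hair motion.
# Body pose copies feed chained AnimDynamics on bangs and side strands, cached for two RigidBody branches.
# The backbling flag selects between them, using a backbling physics asset and shrunken hair bones when worn.
# A front-end idle curve blends DefaultSlot onto large bangs; skydiving events set GravityOverride, which the graph never reads.
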AnimBlueprint /BRCosmetics/Characters/Player/Female/Medium/Bodies/F_MED_LadderDoor/Meshes/Parts/F_MED_LadderDoor_FaceAcc_AnimBP
target_skeleton=&/F_MED_LadderDoor_FaceAcc_Skeleton
graphs [
 AnimGraph {nodes [
   n1 = RigidBody!(ComponentPose=Local To Component(LocalPose=Use cached pose 'startPose'()), Alpha, OverrideWorldGravity:Vector="0,0,-980", ExternalForce) -> (Pose) {
    bClampLinearTranslationLimitToRefPose=true
    bOverrideWorldGravity=true
    overridePhysicsAsset=&/F_MED_LadderDoor_FaceAcc_Physics_Backbling
    simulationSpace=BaseBoneSpace
    baseBoneRef {boneName=root}
    overrideWorldGravity {z=-980}
    simSpaceSettings {
     dampingAlpha=0
     maxAngularAcceleration=50
     maxAngularVelocity=6
     maxLinearAcceleration=3000
     maxLinearVelocity=800
     velocityScaleZ=0.35
     worldAlpha=1
     externalLinearDragV=<1.5 1.5 1.5>
    }
   }
   n2 = "Transform (Modify) Bone - Bone: dyn_L_hair_1"!(ComponentPose=@n1, Alpha=1) -> (Pose) {scaleMode=BMM_Additive, scaleSpace=BCS_BoneSpace, boneToModify {boneName=dyn_L_hair_1}, scale=<0.65 0.85 0.85>}
   n3 = "Transform (Modify) Bone - Bone: dyn_R_hair_1"!(ComponentPose=@n2, Alpha=1) -> (Pose) {scaleMode=BMM_Additive, scaleSpace=BCS_BoneSpace, boneToModify {boneName=dyn_R_hair_1}, scale=<0.65 0.85 0.85>}
   n4 = RigidBody(ComponentPose=Local To Component(LocalPose=Use cached pose 'startPose'()), OverrideWorldGravity="0,0,-980") {
    bClampLinearTranslationLimitToRefPose=true
    bOverrideWorldGravity=true
    overridePhysicsAsset=&/F_MED_LadderDoor_FaceAcc_Physics
    simulationSpace=BaseBoneSpace
    baseBoneRef {boneName=root}
    overrideWorldGravity {z=-980}
    simSpaceSettings {
     dampingAlpha=0
     maxAngularAcceleration=50
     maxAngularVelocity=6
     maxLinearAcceleration=3000
     maxLinearVelocity=800
     velocityScaleZ=0.35
     worldAlpha=1
     externalLinearDragV=<1.5 1.5 1.5>
    }
   }
   n5 = Layered blend per bone(
    BasePose=Blend Poses by bool(
     BlendPose_0=Component To Local(ComponentPose=@n3)
     BlendPose_1=Component To Local(ComponentPose=@n4)
     bActiveValue=Get Is Wearing Backbling()
    )
    BlendPoses_0=Slot 'DefaultSlot'()
    BlendWeights_0=Get customIdlePlayingAlpha() -> (customIdlePlayingAlpha:float)
   ) {layerSetup [{branchFilters [{boneName=dyn_largeBang_1}, {boneName=dyn_largeBang_2}, {boneName=dyn_largeBang_3}, {boneName=dyn_largeBang_4}, {boneName=dyn_largeBang_5}]}]}
   Output Pose(Result=@n5) {layerGroup=None}
   n6 = Copy Pose From Mesh(SourceMeshComponent=Get SkelMesh()) {bUseAttachedParent=true}
   n7 = Copy Pose From Mesh() {bUseAttachedParent=true}
   n8 = Anim Dynamics - Bone: dyn_smallBang(
    ComponentPose=Local To Component(LocalPose=Layered blend per bone(BasePose=@n7, BlendPoses_0=@n6) {layerSetup [{branchFilters [{boneName=upperarm_l}, {boneName=upperarm_r}]}]})
    Alpha=1
    ExternalForce="0,0,0"
   ) {
    angularDampingOverride=1
    angularSpringConstant=7000
    bAngularSpring=true
    bOverrideAngularDamping=true
    bUseGravityOverride=true
    gravityScale=0.35
    simulationSpace=RootRelative
    boundBone {boneName=dyn_smallBang}
    componentLinearAccScale=<1 1 1>
    componentLinearVelScale=<1 1 1>
    gravityOverride {z=-980}
    physicsBodyDefinitions [{
      boundBone {boneName=dyn_smallBang}
      boxExtents=<21 21 21>
      constraintSetup {angularLimitsMax=<0 15 15>, angularLimitsMin=<0 -15 0>, angularTarget=<1 0 0>}
      localJointOffset=<4 0 0>
     }]
   }
   n9 = Anim Dynamics - Bone: dyn_largeBang_1(ComponentPose=@n8, Alpha=1, ExternalForce="0,0,0") {
    angularDampingOverride=1
    angularSpringConstant=7000
    bAngularSpring=true
    bOverrideAngularDamping=true
    bUseGravityOverride=true
    gravityScale=0.35
    simulationSpace=RootRelative
    boundBone {boneName=dyn_largeBang_1}
    componentLinearAccScale=<1 1 1>
    componentLinearVelScale=<1 1 1>
    gravityOverride {z=-980}
    physicsBodyDefinitions [{
      boundBone {boneName=dyn_largeBang_1}
      boxExtents=<21 21 21>
      constraintSetup {angularLimitsMax=<15 0 30>, angularLimitsMin=<-15 -30 -30>, angularTarget=<1 0 0>}
      localJointOffset=<2 0 0>
     }]
   }
   n10 = Anim Dynamics - Bone: dyn_largeBang_2(ComponentPose=@n9, Alpha=1, ExternalForce="0,0,0") {
    angularDampingOverride=1
    angularSpringConstant=7000
    bAngularSpring=true
    bOverrideAngularDamping=true
    bUseGravityOverride=true
    gravityScale=0.35
    simulationSpace=RootRelative
    boundBone {boneName=dyn_largeBang_2}
    componentLinearAccScale=<1 1 1>
    componentLinearVelScale=<1 1 1>
    gravityOverride {z=-980}
    physicsBodyDefinitions [{
      boundBone {boneName=dyn_largeBang_2}
      boxExtents=<21 21 21>
      constraintSetup {angularLimitsMax=<15 0 30>, angularLimitsMin=<-15 -30 -30>, angularTarget=<1 0 0>}
      localJointOffset=<2 0 0>
     }]
   }
   n11 = Anim Dynamics - Bone: dyn_largeBang_3(ComponentPose=@n10, Alpha=1, ExternalForce="0,0,0") {
    angularDampingOverride=1
    angularSpringConstant=7000
    bAngularSpring=true
    bOverrideAngularDamping=true
    bUseGravityOverride=true
    gravityScale=0.35
    simulationSpace=RootRelative
    boundBone {boneName=dyn_largeBang_3}
    componentLinearAccScale=<1 1 1>
    componentLinearVelScale=<1 1 1>
    gravityOverride {z=-980}
    physicsBodyDefinitions [{
      boundBone {boneName=dyn_largeBang_3}
      boxExtents=<21 21 21>
      constraintSetup {angularLimitsMax=<15 10 30>, angularLimitsMin=<-15 -20 -30>, angularTarget=<1 0 0>}
      localJointOffset=<2 0 0>
     }]
   }
   n12 = Anim Dynamics - Bone: dyn_largeBang_4(ComponentPose=@n11, Alpha=1, ExternalForce="0,0,0") {
    angularDampingOverride=1
    angularSpringConstant=7000
    bAngularSpring=true
    bOverrideAngularDamping=true
    bUseGravityOverride=true
    gravityScale=0.35
    simulationSpace=RootRelative
    boundBone {boneName=dyn_largeBang_4}
    componentLinearAccScale=<1 1 1>
    componentLinearVelScale=<1 1 1>
    gravityOverride {z=-980}
    physicsBodyDefinitions [{
      boundBone {boneName=dyn_largeBang_4}
      boxExtents=<21 21 21>
      constraintSetup {angularLimitsMax=<15 10 30>, angularLimitsMin=<-15 -20 -30>, angularTarget=<1 0 0>}
      localJointOffset=<2 0 0>
     }]
   }
   n13 = Anim Dynamics - Bone: dyn_largeBang_5(ComponentPose=@n12, Alpha=1, ExternalForce="0,0,0") {
    angularDampingOverride=1
    angularSpringConstant=7000
    bAngularSpring=true
    bOverrideAngularDamping=true
    bUseGravityOverride=true
    gravityScale=0.35
    simulationSpace=RootRelative
    boundBone {boneName=dyn_largeBang_5}
    componentLinearAccScale=<1 1 1>
    componentLinearVelScale=<1 1 1>
    gravityOverride {z=-980}
    physicsBodyDefinitions [{
      boundBone {boneName=dyn_largeBang_5}
      boxExtents=<21 21 21>
      constraintSetup {angularLimitsMax=<15 10 30>, angularLimitsMin=<-15 -20 -30>, angularTarget=<1 0 0>}
      localJointOffset=<2 0 0>
     }]
   }
   n14 = Anim Dynamics - Bone: dyn_L_sideStrand_1(ComponentPose=@n13, Alpha=1, ExternalForce="0,0,0") {
    angularDampingOverride=1
    angularSpringConstant=5000
    bAngularSpring=true
    bOverrideAngularDamping=true
    bUseGravityOverride=true
    gravityScale=0.35
    simulationSpace=RootRelative
    boundBone {boneName=dyn_L_sideStrand_1}
    componentLinearAccScale=<1 1 1>
    componentLinearVelScale=<1 1 1>
    gravityOverride {z=-980}
    physicsBodyDefinitions [{
      boundBone {boneName=dyn_L_sideStrand_1}
      boxExtents=<21 21 21>
      constraintSetup {angularLimitsMax=<15 45 30>, angularLimitsMin=<-15 -45 -1>, angularTarget=<1 0 0>}
      localJointOffset=<3 0 0>
     }]
   }
   n15 = Anim Dynamics - Bone: dyn_L_sideStrand_2(ComponentPose=@n14, Alpha=1, ExternalForce="0,0,0") {
    angularDampingOverride=1
    angularSpringConstant=5000
    bAngularSpring=true
    bOverrideAngularDamping=true
    bUseGravityOverride=true
    gravityScale=0.35
    simulationSpace=RootRelative
    boundBone {boneName=dyn_L_sideStrand_2}
    componentLinearAccScale=<1 1 1>
    componentLinearVelScale=<1 1 1>
    gravityOverride {z=-980}
    physicsBodyDefinitions [{
      boundBone {boneName=dyn_L_sideStrand_2}
      boxExtents=<21 21 21>
      constraintSetup {angularLimitsMax=<15 45 30>, angularLimitsMin=<-15 -45 -1>, angularTarget=<1 0 0>}
      localJointOffset=<3 0 0>
     }]
   }
   n16 = Anim Dynamics - Bone: dyn_R_sideStrand_1(ComponentPose=@n15, Alpha=1, ExternalForce="0,0,0") {
    angularDampingOverride=1
    angularSpringConstant=5000
    bAngularSpring=true
    bOverrideAngularDamping=true
    bUseGravityOverride=true
    gravityScale=0.35
    simulationSpace=RootRelative
    boundBone {boneName=dyn_R_sideStrand_1}
    componentLinearAccScale=<1 1 1>
    componentLinearVelScale=<1 1 1>
    gravityOverride {z=-980}
    physicsBodyDefinitions [{
      boundBone {boneName=dyn_R_sideStrand_1}
      boxExtents=<21 21 21>
      constraintSetup {angularLimitsMax=<15 45 30>, angularLimitsMin=<-15 -45 -1>, angularTarget=<1 0 0>}
      localJointOffset=<3 0 0>
     }]
   }
   n17 = Anim Dynamics - Bone: dyn_R_sideStrand_2(ComponentPose=@n16, Alpha=1, ExternalForce="0,0,0") {
    angularDampingOverride=1
    angularSpringConstant=5000
    bAngularSpring=true
    bOverrideAngularDamping=true
    bUseGravityOverride=true
    gravityScale=0.35
    simulationSpace=RootRelative
    boundBone {boneName=dyn_R_sideStrand_2}
    componentLinearAccScale=<1 1 1>
    componentLinearVelScale=<1 1 1>
    gravityOverride {z=-980}
    physicsBodyDefinitions [{
      boundBone {boneName=dyn_R_sideStrand_2}
      boxExtents=<21 21 21>
      constraintSetup {angularLimitsMax=<15 45 30>, angularLimitsMin=<-15 -45 -1>, angularTarget=<1 0 0>}
      localJointOffset=<3 0 0>
     }]
   }
   n18 = Anim Dynamics - Bone: dyn_R_sideStrand_3(ComponentPose=@n17, Alpha=1, ExternalForce="0,0,0") {
    angularDampingOverride=1
    angularSpringConstant=5000
    bAngularSpring=true
    bOverrideAngularDamping=true
    bUseGravityOverride=true
    gravityScale=0.35
    simulationSpace=RootRelative
    boundBone {boneName=dyn_R_sideStrand_3}
    componentLinearAccScale=<1 1 1>
    componentLinearVelScale=<1 1 1>
    gravityOverride {z=-980}
    physicsBodyDefinitions [{
      boundBone {boneName=dyn_R_sideStrand_3}
      boxExtents=<21 21 21>
      constraintSetup {angularLimitsMax=<15 45 30>, angularLimitsMin=<-15 -45 -1>, angularTarget=<1 0 0>}
      localJointOffset=<3 0 0>
     }]
   }
   Save cached pose 'startPose'(Pose=Component To Local(ComponentPose=@n18)) {cacheName=startPose}
  ]}
 EventGraph {nodes [
   Event BlueprintUpdateAnimation()
   > Branch(Condition=Get bIsInFrontEnd())
   > Set customIdlePlayingAlpha(customIdlePlayingAlpha:float=GetCurveValue(self=GetOwnerFrontEndAnimInstance(), CurveName=Frontend_Idle_Bypass)) -> (Output_Get:float)
   Event OnCharacterCustomizationLoaded()
   > n1 = Sequence()
   Set GravityOverride(execute=[@n2.then, @n1.then_0], GravityOverride="0,0,-980")
   Set Is Wearing Backbling(execute=@n1.then_1, Is Wearing Backbling=CheckForBackbling())
   Event OnBeginSkydiving()
   > Set GravityOverride(GravityOverride="0,0,600")
   n2 = Event OnEndSkydiving()
   Event BlueprintInitializeAnimation()
   > Set SkelMesh(SkelMesh=GetOwnerSkeletalMeshComponent())
  ]}
]
variables [
 customIdlePlayingAlpha {type=number}
 GravityOverride {type=Vector}
 Is Wearing Backbling {type=boolean}
 SkelMesh {type=SkeletalMeshComponent}
]
